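# Looping SoundCue for wind field splines, named for playback at the spline's closest point, providing the wind heard from outside.
# A Mixer blends two looping wave layers, close and distant, with the close layer weighted slightly louder.
# Each layer passes through its own attenuation asset before mixing, so distance shapes the near and far textures independently.
# The cue routes to the Fort_Gadgets sound class at reduced overall volume.
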SoundCue /Wind/Gameplay/WindFields/Audio/WindSpline_ClosestPt_Loop_Cue
root=Mixer_0
nodes [
 {
  id=Attenuation_0
  children [WavePlayer_0]
  properties {attenuationSettings=/Wind/Gameplay/WindFields/Audio/Settings/WindSpline_ClosestPoint_Loop_Distant_Att}
 }
 {
  id=Attenuation_1
  children [WavePlayer_1]
  properties {attenuationSettings=/Wind/Gameplay/WindFields/Audio/Settings/WindSpline_ClosestPoint_Loop_Close_Att}
 }
 {id=Mixer_0, children [Attenuation_1, Attenuation_0], properties {inputVolume [1.2,1]}}
 {id=WavePlayer_0, properties {bLooping=true, soundWaveAssetPtr=/Wind/Gameplay/WindFields/Audio/Waves/WindSpline_Outside_Distant_Loop_01}}
 {id=WavePlayer_1, properties {bLooping=true, soundWaveAssetPtr=/Wind/Gameplay/WindFields/Audio/Waves/WindSpline_Outside_Close_Loop_01}}
]
properties {maxDistance=7000, soundClassObject=&Fort_Gadgets, volumeMultiplier=0.425}
