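# MetaSound patch that wraps float A toward the range -B to B, with a default B of 180 that suggests angles in degrees.
# Two CompareFloat checks test A against B and -B, and their boolean results convert to float multipliers.
# These gate offsets of -2B and +2B that are added to A, giving a branchless single-step wrap rather than a full modulo.
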MetaSoundPatch /FortMetasound/Patches/WrapFloat
inputs [
 A {type=Float}
 B {type=Float}
]
interfaces [MetaSound:1.0]
outputs [Out {type=Float}]
pages [Default {
  input_defaults {A=0, B=180}
  nodes [
   n1 = Convert.Float:Bool(Bool=&/CompareFloat(A=^A, B=^B) {also {defaults {Type=3}}}.Result)
   n2 = UE.Multiply:Float(PrimaryOperand=^B) {also {defaults {AdditionalOperands=-2}}}
   n3 = &CompareFloat(A=^A, B=UE.Multiply:Float(PrimaryOperand=^B) {also {defaults {AdditionalOperands=-1}}}.Out) {also {defaults {Type=2}}}
   Add_1 = UE.Add:Float(
    PrimaryOperand=UE.Add:Float(PrimaryOperand=^A, AdditionalOperands=UE.Multiply:Float(PrimaryOperand=@n1.Float, AdditionalOperands=@n2.Out).Out).Out
    AdditionalOperands=UE.Multiply:Float(
     PrimaryOperand=Convert.Float:Bool(Bool=@n3.Result).Float
     AdditionalOperands=UE.Multiply:Float(PrimaryOperand=^B) {also {defaults {AdditionalOperands=2}}}.Out
    ).Out
   )
   InitVariable.Float()
   InitVariable.Float()
   InitVariable.Float()
   InitVariable.Float()
  ]
  links [{dst="@Out", src=Add_1.Out}]
 }]
