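# Looping stereo MetaSound source for wind heard by a grounded player inside a wind spline field.
# The wrapped player-minus-wind angle drives azimuth mappers that mix facing-away and facing-towards loops, pitch-shifting the waveshaped towards loop.
# The mix passes through an azimuth panner and a biquad filter with angle-mapped cutoff.
# An on-play attack envelope multiplied by smoothed, range-mapped Attenuation sets the output level.
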MetaSoundSource /Wind/Gameplay/WindFields/Audio/WindSpline_Inside_Grounded_New_MSS
inputs [
 Attenuation {type=Float}
 InsideFacingAway {type=WaveAsset}
 InsideFacingTowards {type=WaveAsset}
 PlayerAngle {type=Float}
 UE.Source.OnPlay {type=Trigger}
 WaveShapeAmount {type=Float}
 WindAngle {type=Float}
]
interfaces [UE.OutputFormat.Stereo:1.0, UE.Source:1.1]
outputs [
 "UE.OutputFormat.Stereo.Audio:0" {type=Audio}
 "UE.OutputFormat.Stereo.Audio:1" {type=Audio}
]
pages [Default {
  input_defaults {
   Attenuation=1
   InsideFacingAway=/Wind/Gameplay/WindFields/Audio/Waves/WindSpline_Inside_Standing_Loop_02
   InsideFacingTowards=/Wind/Gameplay/WindFields/Audio/Waves/WindSpline_Inside_Standing_Loop_03
   PlayerAngle=-180
   UE.Source.OnPlay=null
   WaveShapeAmount=1.5
   WindAngle=0
  }
  nodes [
   n1 = UE.Subtract:Float(PrimaryOperand=^PlayerAngle, AdditionalOperands=^WindAngle)
   n2 = VariableAccessor.Float(Variable=VariableMutator.Float(Variable=InitVariable.Float().Variable, Value=/FortMetasound/Patches/WrapFloat(A=@n1.Out).Out).Variable)
   n3 = VariableAccessor.Float(Variable=@n2.Variable)
   n4 = VariableAccessor.Float(Variable=@n3.Variable)
   n5 = /FortMetasound/Patches/MSP_AzimuthMapper(Azimuth=VariableAccessor.Float(Variable=@n4.Variable).Value) {also {defaults {Back=1, Left=0.75, Right=0.75}}}
   WavePlayer_1 = UE.Wave Player:Stereo(Wave Asset=^InsideFacingAway, Play=^UE.Source.OnPlay) {also {defaults {Loop="<MetaSoundBuilderResult.SUCCEEDED: 0>"}, version [1,1]}}
   FrequencyMultipliertoSemitone_0 = UE.Frequency Multiplier to Semitone:Float(Frequency Multiplier=&MSP_AzimuthMapper(Azimuth=@n2.Value) {also {defaults {Back=1.2, Front=0.8, Left=1}}}.Out)
   WavePlayer_0 = UE.Wave Player:Stereo(Wave Asset=^InsideFacingTowards, Play=^UE.Source.OnPlay, Pitch Shift=@FrequencyMultipliertoSemitone_0.Semitones) {also {defaults {Loop="<MetaSoundBuilderResult.SUCCEEDED: 0>"}, version [1,1]}}
   AudioMixerAudioMixerStereo2_0 = "AudioMixer.Audio Mixer (Stereo, 2)"(
    Gain 1=Max.Max:Float(B=@n5.Out) {also {defaults {A=0.2}}}.Value
    Gain 0=UE.Subtract:Float(AdditionalOperands=@n5.Out) {also {defaults {PrimaryOperand=1}}}.Out
    In 0 L=@WavePlayer_1.Out Left
    In 0 R=@WavePlayer_1.Out Right
    In 1 R=UE.WaveShaper:Audio(Amount=^WaveShapeAmount, In=@WavePlayer_0.Out Right).Out
    In 1 L=UE.WaveShaper:Audio(Amount=^WaveShapeAmount, In=@WavePlayer_0.Out Left).Out
   )
   n6 = &AzimuthPannerStereo(In L=@AudioMixerAudioMixerStereo2_0.Out L, In R=@AudioMixerAudioMixerStereo2_0.Out R, Azimuth=@n3.Value) {also {defaults {StereoSpread=45}}}
   n7 = &StereoBiquadFilter(
    In L=@n6.Out L
    In R=@n6.Out R
    Cutoff Frequency=&MSP_AzimuthMapper(Azimuth=@n4.Value) {also {defaults {Back=20000, Front=7000, Left=20000, Right=20000}}}.Out
   )
   ADSREnvelopeADSREnvelope_0 = ADSR Envelope.ADSR Envelope:Float(Trigger Attack=^UE.Source.OnPlay) {also {defaults {Attack Time=1, Sustain Level=1}}}
   n8 = MapRange.MapRange:Float(In=UE.InterpTo:Audio(Target=^Attenuation) {also {defaults {Interp Time=0.5}}}.Value) {also {defaults {Out Range A=0.25}}}
   n9 = VariableAccessor.Float(Variable=VariableMutator.Float(
     Variable=InitVariable.Float().Variable
     Value=UE.Multiply:Float(PrimaryOperand=@ADSREnvelopeADSREnvelope_0.Out Envelope, AdditionalOperands=@n8.Out Value).Out
    ).Variable)
   Multiply_0 = UE.Multiply:Audio by Float(PrimaryOperand=@n7.Out R, AdditionalOperands=@n9.Value)
   Multiply_1 = UE.Multiply:Audio by Float(PrimaryOperand=@n7.Out L, AdditionalOperands=@n9.Value)
  ]
  links [{dst="@UE.OutputFormat.Stereo.Audio:1", src=Multiply_0.Out}, {dst="@UE.OutputFormat.Stereo.Audio:0", src=Multiply_1.Out}]
 }]
properties {
 assetImportData=&self:AssetImportData
 bLooping=true
 duration=10000
 numChannels=2
 outputFormat=Stereo
 sampleRate=48000
 soundClassObject=&Fort_Gadgets
 volume=0.5
}
