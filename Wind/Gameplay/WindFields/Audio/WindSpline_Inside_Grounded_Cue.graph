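# Looping wind spline sound cue in the gadgets sound class, named for a player inside the wind field while grounded.
# A quality level node at the root picks the branch: the first two levels loop WindSpline_Inside_Grounded_New_MSS, apparently a MetaSound.
# The other two levels loop the Standing_Loop wave through an Enveloper that fades volume in over two seconds.
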SoundCue /Wind/Gameplay/WindFields/Audio/WindSpline_Inside_Grounded_Cue
root=QualityLevel_0
nodes [
 {
  id=Enveloper_0
  children [WavePlayer_0]
  properties {pitchCurve {editorCurveData {keys [{time=0}]}}, volumeCurve {editorCurveData {keys [<at 0 0>,<at 2 1>]}}}
 }
 {id=QualityLevel_0, children [WavePlayer_1,WavePlayer_1,Enveloper_0,Enveloper_0]}
 {id=WavePlayer_0, properties {bLooping=true, soundWaveAssetPtr=/Wind/Gameplay/WindFields/Audio/Waves/WindSpline_Inside_Standing_Loop_01}}
 {id=WavePlayer_1, properties {bLooping=true, soundWaveAssetPtr=&/WindSpline_Inside_Grounded_New_MSS}}
]
properties {maxDistance=2097152, soundClassObject=&Fort_Gadgets, volumeMultiplier=1}
