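# Sound cue for the one-shot that plays when a grounded state inside a wind spline field ends, using standing-exit waves.
# A Random node picks evenly between two exit waves and feeds a Modulator at the root.
# It routes through the Fort_Gadgets sound class, with attenuation and concurrency settings named for wind spline enter and exit sounds.
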SoundCue /Wind/Gameplay/WindFields/Audio/WindSpline_Inside_Grounded_End_Cue
root=Modulator_0
nodes [
 {id=Modulator_0, children [Random_0]}
 {id=Random_0, children [WavePlayer_1,WavePlayer_2], properties {weights [1,1]}}
 {id=WavePlayer_1, properties {soundWaveAssetPtr=/Wind/Gameplay/WindFields/Audio/Waves/WindSpline_Inside_Standing_Exit_01}}
 {id=WavePlayer_2, properties {soundWaveAssetPtr=/Wind/Gameplay/WindFields/Audio/Waves/WindSpline_Inside_Standing_Exit_02}}
]
properties {
 attenuationSettings=/Wind/Gameplay/WindFields/Audio/Settings/WindSpline_EnterExit_Att
 maxDistance=2000
 soundClassObject=&Fort_Gadgets
 volumeMultiplier=1.0594
 concurrencySet [/Wind/Gameplay/WindFields/Audio/Settings/WindSpline_EnterExit_Conc]
}
