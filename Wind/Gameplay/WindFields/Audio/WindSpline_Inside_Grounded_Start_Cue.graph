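# Sound cue for a grounded player starting to be inside a wind spline, as suggested by its name and its standing-enter waves.
# A Random node picks between two enter wave variations with equal weight, feeding a Modulator at the root for pitch and volume variation.
# It plays in the gadgets sound class, using the wind spline enter/exit attenuation and concurrency settings.
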SoundCue /Wind/Gameplay/WindFields/Audio/WindSpline_Inside_Grounded_Start_Cue
root=Modulator_0
nodes [
 {id=Modulator_0, children [Random_1]}
 {id=Random_1, children [WavePlayer_0,WavePlayer_3], properties {weights [1,1]}}
 {id=WavePlayer_0, properties {soundWaveAssetPtr=/Wind/Gameplay/WindFields/Audio/Waves/WindSpline_Inside_Standing_Enter_01}}
 {id=WavePlayer_3, properties {soundWaveAssetPtr=/Wind/Gameplay/WindFields/Audio/Waves/WindSpline_Inside_Standing_Enter_02}}
]
properties {
 attenuationSettings=/Wind/Gameplay/WindFields/Audio/Settings/WindSpline_EnterExit_Att
 maxDistance=2000
 priority=2
 soundClassObject=&Fort_Gadgets
 volumeMultiplier=1.0594
 concurrencySet [/Wind/Gameplay/WindFields/Audio/Settings/WindSpline_EnterExit_Conc]
}
