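# Sound cue for the start of being airborne inside a wind spline, using skydiving enter sounds.
# A Random node picks between two equally weighted wave variations and feeds a Modulator at the root.
# It plays in the Fort_Gadgets sound class, with attenuation and concurrency assets named for wind spline enter/exit.
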SoundCue /Wind/Gameplay/WindFields/Audio/WindSpline_Inside_InAir_Start_Cue
root=Modulator_0
nodes [
 {id=Modulator_0, children [Random_0]}
 {id=Random_0, children [WavePlayer_2,WavePlayer_3], properties {weights [1,1]}}
 {id=WavePlayer_2, properties {soundWaveAssetPtr=/Wind/Gameplay/WindFields/Audio/Waves/WindSpline_Inside_Skydiving_Enter_02}}
 {id=WavePlayer_3, properties {soundWaveAssetPtr=/Wind/Gameplay/WindFields/Audio/Waves/WindSpline_Inside_Skydiving_Enter_01}}
]
properties {
 attenuationSettings=/Wind/Gameplay/WindFields/Audio/Settings/WindSpline_EnterExit_Att
 maxDistance=2000
 priority=2
 soundClassObject=&Fort_Gadgets
 volumeMultiplier=1.0594
 concurrencySet [/Wind/Gameplay/WindFields/Audio/Settings/WindSpline_EnterExit_Conc]
}
