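# Translucent unlit spline-mesh material for line streaks inside wind tubes, tinted by a time-of-day curve atlas.
# Opacity combines panning noise, contrast and UV shaping with position mask, depth, camera-distance, time-of-day and vertex-alpha fades.
# World position offset uses SplineThicken for width plus scrolling vector noise for wobble, which is disabled on low quality, mobile and forward shading.
# Vertex red and green vary width, speed and noise per line.
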Material /Wind/Gameplay/WindFields/FX/M_WindSpline_InnerTube_Lines
graph {
 properties {
  bUsedWithNanite=true
  bUsedWithSplineMeshes=true
  blendMode=BLEND_Translucent
  pixelDepthOffsetMode=PDOM_Legacy
  shadingModel=MSM_Unlit
  twoSided=true
  usedShadingModels=Unlit
 }
 nodes [
  AppendVector(A=TextureCoordinate(), B=Constant())
  n1 = CurveAtlasRowParameter(CurveTime=CollectionParameter() {collection=&FortniteMaterialParameters, parameterName=TimeOfDay_0-1}) {
   atlas=&AthenaCurveAtlas01
   curve=/Game/Athena/LinearColorCurves/Curve_TimeOfDay_S24
   defaultValue=329
   group=TimeOfDay
   parameterName=TOD_Curve
  }
  MaterialFunctionCall(Emissive=Multiply(A=@n1, B=ScalarParameter() {defaultValue=1, parameterName=TOD Brightness Scale, sliderMax=1})) {materialFunction=&MF_FX_EmissiveAdjustments}
  n2 = VertexColor()
  n3 = Subtract(A=Multiply(A=NamedRerouteUsage(), B=NamedRerouteUsage()), B=LinearInterpolate(Alpha=@n2.G) {constA=0.1, constB=0})
  n4 = LinearInterpolate(Alpha=@n2.G) {constA=0.25, constB=0}
  n5 = LinearInterpolate(Alpha=Multiply(
    A=MaterialFunctionCall(In=@n3, Contrast=@n4) {materialFunction=&CheapContrast}.Result
    B=Constant() {r=1}
   )) {constA=0.1}
  n6 = TextureCoordinate()
  n7 = Sine(None=Multiply(A=MaterialFunctionCall(Float2=@n6) {materialFunction=&BreakOutFloat2Components}.R) {constB=0.5})
  n8 = MaterialFunctionCall() {materialFunction=&LinearGradient}
  n9 = Multiply(A=Power(Base=Saturate(None=@n7)) {constExponent=3}, B=Multiply(A=@n8.VGradient, B=OneMinus(None=@n8.VGradient)))
  n10 = MaterialFunctionCall(Distance=ScalarParameter() {defaultValue=500, parameterName=CameraDistanceClose_Fade}) {materialFunction=&Near_Camera_Fade}
  n11 = VertexColor()
  n12 = ScalarParameter() {defaultValue=1000, parameterName=DepthFadeDistance}
  n13 = Constant() {r=1}
  n14 = Multiply(
   A=Multiply(A=Multiply(A=@n9, B=@n10.Result), B=LinearInterpolate(Alpha=@n11.G) {constA=0.5})
   B=QualitySwitch(Default=DepthFade(FadeDistance=@n12), Low=@n13, Medium=@n13)
  )
  n15 = Multiply(
   A=Multiply(
    A=MaterialFunctionCall() {materialFunction=/Game/Athena/Environments/Blueprints/SkyTube/FX/MF_WindSpline_PositionMask}.Result
    B=Multiply(A=@n5, B=@n14)
   )
   B=MaterialFunctionCall(
    Value_Morning=ScalarParameter() {defaultValue=1, group=ToD Opacity, parameterName=OpacityMasterMorning}
    Value_Day=ScalarParameter() {defaultValue=1, group=ToD Opacity, parameterName=OpacityMasterDay}
    Value_Evening=ScalarParameter() {defaultValue=1, group=ToD Opacity, parameterName=OpacityMasterEvening}
    Value_Night=ScalarParameter() {defaultValue=1, group=ToD Opacity, parameterName=OpacityMasterNight}
   ) {materialFunction=&Scalar_Time_of_Day}.Result
  )
  n16 = ScalarParameter() {defaultValue=1, group=ToD Opacity, parameterName=OpacityGlobalMultiplier}
  n17 = VertexColor()
  n18 = ScalarParameter() {defaultValue=5, group=Opacity, parameterName=Vertex Alpha Exponent}
  n19 = ScalarParameter() {defaultValue=18000, parameterName=NearCameraFade}
  Multiply(
   A=Multiply(A=Multiply(A=@n15, B=@n16), B=Power(Base=@n17.A, Exp=@n18) {constExponent=5})
   B=OneMinus(None=MaterialFunctionCall(Distance=@n19) {materialFunction=&Near_Camera_Fade}.Result)
  )
  n20 = VectorParameter() {
   group=10.UVDistortNoise_B
   parameterName=NoiseB_UVTransform
   channelNames {a=PannerV, b=PannerU, g=ScaleV, r=ScaleU}
   defaultValue=<rgba 1 1 0.2 0>
  }
  n21 = VertexColor()
  n22 = Add(A=Multiply(A=TextureCoordinate(), B=AppendVector(A=@n20.ScaleU, B=@n20.ScaleU)), B=AppendVector(A=@n21.R, B=@n21.R))
  n23 = Multiply(A=AppendVector(A=@n20.PannerU, B=@n20.PannerU), B=LinearInterpolate(Alpha=VertexColor().G) {constA=0.5})
  NamedRerouteDeclaration(None=TextureSampleParameter2D(UVs=Panner(Coordinate=@n22, Speed=@n23) {speedX=0.216}) {group=10.UVDistortNoise_B, parameterName=NoiseB_Texture, texture=&AtmosphericCloudNoise01}.RGB) {name=R_NoiseB_1, nodeColor=<rgba 1 0.7969 0 1>}
  n24 = VectorParameter() {
   group=10.UVDistortNoise_A
   parameterName=NoiseA_UVTransform
   channelNames {a=PannerV, b=PannerU, g=ScaleV, r=ScaleU}
   defaultValue=<rgba 3 1 0.2 0>
  }
  n25 = Multiply(A=TextureCoordinate(), B=AppendVector(A=@n24.ScaleU, B=@n24.ScaleU))
  n26 = VertexColor()
  n27 = AppendVector(A=@n26.R, B=@n26.R)
  n28 = LinearInterpolate(Alpha=VertexColor().G) {constA=0.5}
  n29 = AppendVector(A=@n24.PannerU, B=@n24.PannerU)
  n30 = TextureSampleParameter2D(UVs=Panner(Coordinate=Add(A=@n25, B=@n27), Speed=Multiply(A=@n29, B=@n28)) {speedX=0.216}) {
   group=10.UVDistortNoise_A
   parameterName=NoiseA_Texture
   samplerType=SAMPLERTYPE_LinearGrayscale
   texture=/Game/Gadgets/Assets/VinderTech_GliderChute/Glider_Thrive_Female/FX/Texture/T_AntiMaterialCharge_Dissolve_01
  }
  NamedRerouteDeclaration(None=@n30.RGB) {name=R_NoiseA, nodeColor=<rgba 0 1 0.25 1>}
  n31 = VectorParameter() {
   group=10.UVDistortNoise_B
   parameterName=NoiseB_UVTransform
   channelNames {a=PannerV, b=PannerU, g=ScaleV, r=ScaleU}
   defaultValue=<rgba 1 1 0.2 0>
  }
  n32 = VertexColor()
  n33 = Add(A=Multiply(A=TextureCoordinate(), B=AppendVector(A=@n31.ScaleU, B=@n31.ScaleU)), B=AppendVector(A=@n32.R, B=@n32.R))
  n34 = Multiply(A=AppendVector(A=@n31.PannerU, B=@n31.PannerU), B=LinearInterpolate(Alpha=VertexColor().G) {constA=0.5})
  NamedRerouteDeclaration(None=TextureSampleParameter2D(UVs=Panner(Coordinate=@n33, Speed=@n34) {speedX=0.216}) {group=10.UVDistortNoise_B, parameterName=NoiseB_Texture, texture=&AtmosphericCloudNoise01}.RGB) {name=R_NoiseB, nodeColor=<rgba 1 0.7969 0 1>}
  n35 = LinearInterpolate(Alpha=VertexColor().R) {constA=0.75}
  n36 = LinearInterpolate(Alpha=VertexColor().G) {constA=0.15, constB=1.5}
  n37 = Multiply(A=Multiply(A=ScalarParameter() {defaultValue=1, parameterName=WidthBase}, B=@n35), B=@n36)
  n38 = Multiply(A=Multiply(A=@n35, B=ScalarParameter() {defaultValue=1, parameterName=WidthTip}), B=@n36)
  n39 = WorldPosition()
  n40 = ObjectPositionWS()
  n41 = VertexColor()
  n42 = Multiply(
   A=Multiply(A=ScalarParameter() {defaultValue=250, group=Movement, parameterName=Curl_Speed}, B=LinearInterpolate(Alpha=@n41.R) {constA=0.75})
   B=LinearInterpolate(Alpha=VertexColor().G) {constA=0.5}
  )
  n43 = Time()
  n44 = Divide(
   A=Subtract(A=Subtract(A=@n39.XYZ, B=@n40), B=Multiply(A=@n42, B=@n43))
   B=ScalarParameter() {defaultValue=500, group=Movement, parameterName=Curl_Scale}
  )
  n45 = Multiply(
   A=LinearInterpolate(Alpha=VertexColor().R) {constA=0.75}
   B=ScalarParameter() {defaultValue=50, group=Movement, parameterName=Curl_Intensity}
  )
  n46 = LinearInterpolate(Alpha=VertexColor().G) {constA=0.5}
  n47 = Constant()
  n48 = ShadingPathSwitch(Default=Multiply(A=VectorNoise(World Position=@n44) {noiseFunction=VNF_VectorALU}, B=Multiply(A=@n45, B=@n46)), Mobile=@n47) {desc=Migrated from FeatureLevelSwitch}
  n49 = QualitySwitch(Default=@n48, Low=@n47, Medium=@n48)
  Reroute(None=Add(
    A=MaterialFunctionCall(WidthBase=@n37, WidthTip=@n38) {materialFunction=&SplineThicken}.WorldPositionOffset
    B=ShadingPathSwitch(Default=@n49, Forward=@n47)
   ))
 ]
}
