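# Translucent unlit spline-mesh material for wind spline tube lines, named for use on the ground, with time-of-day color and opacity.
# Vertex color randomizes or scales SplineThicken widths, curl-noise vertex motion (zeroed on cheaper paths) and noise panning.
# Two panned noises, UV gradient masks, camera and depth fades and a wind-spline position mask shape opacity.
# A height-offset branch uses the Apollo landscape runtime virtual texture to conform vertices to terrain.
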Material /Wind/Gameplay/WindFields/FX/M_WindSpline_InnerTube_LinesOnGround
graph {
 properties {
  bUsedWithSplineMeshes=true
  blendMode=BLEND_Translucent
  pixelDepthOffsetMode=PDOM_Legacy
  shadingModel=MSM_Unlit
  twoSided=true
  usedShadingModels=Unlit
 }
 nodes [
  n1 = Subtract(A=WorldPosition().XYZ, B=ObjectPositionWS())
  n2 = ScalarParameter() {defaultValue=250, group=Movement, parameterName=Curl_Speed}
  n3 = LinearInterpolate(Alpha=VertexColor().R) {constA=0.75}
  n4 = VertexColor()
  n5 = Multiply(A=Multiply(A=Multiply(A=@n2, B=@n3), B=LinearInterpolate(Alpha=@n4.G) {constA=0.5}), B=Time())
  n6 = VectorNoise(World Position=Divide(A=Subtract(A=@n1, B=@n5), B=ScalarParameter() {defaultValue=500, group=Movement, parameterName=Curl_Scale})) {noiseFunction=VNF_VectorALU}
  n7 = VertexColor()
  n8 = Multiply(
   A=Multiply(A=LinearInterpolate(Alpha=@n7.R) {constA=0.75}, B=ScalarParameter() {defaultValue=50, group=Movement, parameterName=Curl_Intensity})
   B=LinearInterpolate(Alpha=VertexColor().G) {constA=0.5}
  )
  n9 = Constant()
  n10 = ShadingPathSwitch(Default=Multiply(A=Constant3Vector() {constant=<rgba 1 1 0 0>}, B=Multiply(A=@n6, B=@n8)), Mobile=@n9) {desc=Migrated from FeatureLevelSwitch}
  n11 = ShadingPathSwitch(Default=QualitySwitch(Default=@n10, Low=@n9, Medium=@n10), Forward=@n9)
  n12 = WorldPosition()
  n13 = ComponentMask(None=ObjectBounds()) {b=true}
  n14 = Add(
   A=Subtract(
    A=RuntimeVirtualTextureSample!(UVs, World Position) {materialType=WorldHeight, virtualTexture=/Game/Athena/Apollo/Maps/Landscape/VT/RVT_Apollo_Landscape_Height}.WorldHeight
    B=ComponentMask(None=@n12.XYZ) {b=true}
   )
   B=@n13
  )
  n15 = Multiply(A=@n13, B=ScalarParameter() {defaultValue=1, parameterName=OffsetHeight})
  Add(A=@n11, B=Multiply(A=Constant3Vector() {constant=<rgba 0 0 1 0>}, B=Subtract(A=@n14, B=@n15)))
  n16 = LinearInterpolate(Alpha=VertexColor().R) {constA=0.75}
  n17 = Multiply(A=ScalarParameter() {defaultValue=1, parameterName=WidthBase}, B=@n16)
  n18 = LinearInterpolate(Alpha=VertexColor().G) {constA=0.15, constB=1.5}
  n19 = Multiply(A=@n16, B=ScalarParameter() {defaultValue=1, parameterName=WidthTip})
  Add(
   A=MaterialFunctionCall(WidthBase=Multiply(A=@n17, B=@n18), WidthTip=Multiply(A=@n19, B=@n18)) {materialFunction=&SplineThicken}.WorldPositionOffset
   B=@n11
  )
  n20 = MaterialFunctionCall(
   Value_Morning=ScalarParameter() {defaultValue=1, parameterName=Color_Morning}
   Value_Day=ScalarParameter() {defaultValue=1, parameterName=Color_Day}
   Value_Evening=ScalarParameter() {defaultValue=1, parameterName=Color_Evening}
   Value_Night=ScalarParameter() {defaultValue=1, parameterName=Color_Night}
  ) {materialFunction=&Scalar_Time_of_Day}
  n21 = MaterialFunctionCall() {materialFunction=/Game/Effects/Fort_Effects/Materials/_Functions/CurveColor_Time_of_Day_S23}
  Clamp(None=Divide(A=Multiply(A=@n20.Result, B=@n21.Result), B=EyeAdaptation().EyeAdaptation)) {maxDefault=4}
  Constant()
  MaterialFunctionCall(
   Color_Morning=VectorParameter() {parameterName="Color_Morning (V3)", defaultValue=<rgba 1 1 1 0>}.RGB
   Color_Day=VectorParameter() {parameterName="Color_Day (V3)", defaultValue=<rgba 1 1 1 0>}.RGB
   Color_Evening=VectorParameter() {parameterName="Color_Evening (V3)", defaultValue=<rgba 1 1 1 0>}.RGB
   Color_Night=VectorParameter() {parameterName="Color_Night (V3)", defaultValue=<rgba 1 1 1 0>}.RGB
  ) {materialFunction=&Color_Time_of_Day}
  MaterialFunctionCall!(Color) {materialFunction=/Game/Packages/Fortress_SharedMaterials/Anime_Master_Material/MF_TODColorCurve}
  MaterialFunctionCall() {materialFunction=/Game/Effects/Fort_Effects/Materials/_Functions/CurveColor_Time_of_Day}
  MaterialFunctionCall() {materialFunction=&Scalar_Time_of_Day}
  n22 = NamedRerouteUsage()
  n23 = NamedRerouteUsage()
  n24 = VertexColor()
  n25 = Subtract(A=Add(A=@n22, B=@n23), B=LinearInterpolate(Alpha=@n24.G) {constA=0.1, constB=0})
  n26 = LinearInterpolate(Alpha=@n24.G) {constA=0.25, constB=0}
  n27 = LinearInterpolate(
   A=ScalarParameter() {parameterName=DarkestOpactiy}
   Alpha=Multiply(
    A=MaterialFunctionCall(In=@n25, Contrast=@n26) {materialFunction=&CheapContrast}.Result
    B=Constant() {r=1}
   )
  ) {constA=0.02}
  n28 = Multiply(A=MaterialFunctionCall(Float2=TextureCoordinate()) {materialFunction=&BreakOutFloat2Components}.R) {constB=0.5}
  n29 = Power(Base=Saturate(None=Sine(None=@n28))) {constExponent=3}
  n30 = MaterialFunctionCall() {materialFunction=&LinearGradient}
  n31 = Multiply(A=@n30.VGradient, B=OneMinus(None=@n30.VGradient))
  n32 = ScalarParameter() {defaultValue=500, parameterName=CameraDistanceClose_Fade}
  n33 = Multiply(
   A=Multiply(
    A=Multiply(A=@n29, B=@n31)
    B=MaterialFunctionCall(Distance=@n32) {materialFunction=&Near_Camera_Fade}.Result
   )
   B=LinearInterpolate(Alpha=VertexColor().G) {constA=0.5}
  )
  Multiply(A=@n27, B=@n33)
  n34 = MaterialFunctionCall(
   Value_Morning=ScalarParameter() {defaultValue=1, group=Opacity, parameterName="Value_Morning (S)"}
   Value_Day=ScalarParameter() {defaultValue=1, group=Opacity, parameterName="Value_Day (S)"}
   Value_Evening=ScalarParameter() {defaultValue=1, group=Opacity, parameterName="Value_Evening (S)"}
   Value_Night=ScalarParameter() {defaultValue=1, group=Opacity, parameterName="Value_Night (S)"}
  ) {materialFunction=&Scalar_Time_of_Day}
  n35 = DepthFade(FadeDistance=ScalarParameter() {defaultValue=1000, parameterName=DepthFadeDistance})
  n36 = Constant() {r=1}
  n37 = Multiply(A=@n27, B=Multiply(A=@n33, B=QualitySwitch(Default=@n35, Low=@n36, Medium=@n36)))
  Multiply(
   A=MaterialFunctionCall() {materialFunction=/Game/Athena/Environments/Blueprints/SkyTube/FX/MF_WindSpline_PositionMask}.Result
   B=Multiply(A=Multiply(A=@n34.Result, B=@n37), B=ScalarParameter() {defaultValue=1, group=Opacity, parameterName=OpacityMaster})
  )
  Multiply(A=@n22, B=@n23)
  n38 = VectorParameter() {
   group=10.UVDistortNoise_A
   parameterName=NoiseA_UVTransform
   channelNames {a=PannerV, b=PannerU, g=ScaleV, r=ScaleU}
   defaultValue=<rgba 3 1 0 0.2>
  }
  n39 = Multiply(A=TextureCoordinate(), B=AppendVector(A=@n38.ScaleU, B=@n38.ScaleU))
  n40 = VertexColor()
  n41 = AppendVector(A=@n40.R, B=@n40.R)
  n42 = AppendVector(A=@n38.PannerU, B=@n38.PannerU)
  n43 = LinearInterpolate(Alpha=VertexColor().G) {constA=0.5}
  n44 = TextureSampleParameter2D(UVs=Panner(Coordinate=Add(A=@n39, B=@n41), Speed=Multiply(A=@n42, B=@n43)) {speedX=0.216}) {
   group=10.UVDistortNoise_A
   parameterName=NoiseA_Texture
   samplerType=SAMPLERTYPE_LinearGrayscale
   texture=/Game/Gadgets/Assets/VinderTech_GliderChute/Glider_Thrive_Female/FX/Texture/T_AntiMaterialCharge_Dissolve_01
  }
  NamedRerouteDeclaration(None=@n44.RGB) {name=R_NoiseA, nodeColor=<rgba 0 1 0.25 1>}
  n45 = VectorParameter() {
   group=10.UVDistortNoise_B
   parameterName=NoiseB_UVTransform
   channelNames {a=PannerV, b=PannerU, g=ScaleV, r=ScaleU}
   defaultValue=<rgba 1 1 0 0.2>
  }
  n46 = VertexColor()
  n47 = Add(A=Multiply(A=TextureCoordinate(), B=AppendVector(A=@n45.ScaleU, B=@n45.ScaleU)), B=AppendVector(A=@n46.R, B=@n46.R))
  n48 = Multiply(A=AppendVector(A=@n45.PannerU, B=@n45.PannerU), B=LinearInterpolate(Alpha=VertexColor().G) {constA=0.5})
  NamedRerouteDeclaration(None=TextureSampleParameter2D(UVs=Panner(Coordinate=@n47, Speed=@n48) {speedX=0.216}) {group=10.UVDistortNoise_B, parameterName=NoiseB_Texture, texture=&AtmosphericCloudNoise01}.RGB) {name=R_NoiseB, nodeColor=<rgba 1 0.7969 0 1>}
  ScalarParameter() {defaultValue=0.25}
 ]
}
