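# Translucent unlit Niagara ribbon material for soft wind trails in the WindFields gameplay effects.
# Emissive is particle color, optionally tinted by a time-of-day curve atlas driven by a parameter collection.
# Opacity multiplies two panning noise textures by tail and top/bottom gradient masks, view-angle and camera-distance fades, and particle alpha.
# A quality switch skips the final depth fade on Low.
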Material /Wind/Gameplay/WindFields/FX/M_Wind_Trail_Soft
graph {
 properties {bUsedWithNiagaraRibbons=true, blendMode=BLEND_Translucent, pixelDepthOffsetMode=PDOM_Legacy, shadingModel=MSM_Unlit, usedShadingModels=Unlit}
 nodes [
  n1 = ParticleColor()
  n2 = CurveAtlasRowParameter(CurveTime=CollectionParameter() {collection=&FortniteMaterialParameters, parameterName=TimeOfDay_0-1}) {
   atlas=&AthenaCurveAtlas01
   curve=/Game/Athena/LinearColorCurves/Curve_TimeOfDay_S24
   defaultValue=329
   group=TimeOfDay
   parameterName=TOD_Curve
  }
  MaterialFunctionCall(Emissive=LinearInterpolate(A=@n1.RGB, B=Multiply(A=@n2, B=@n1.RGB), Alpha=ScalarParameter() {parameterName=TOD Weight, sliderMax=1})) {materialFunction=&MF_FX_EmissiveAdjustments}
  n3 = MaterialFunctionCall() {materialFunction=&LinearGradient}
  n4 = OneMinus(None=@n3.UGradient)
  Power(Base=@n4) {constExponent=1}
  n5 = TextureCoordinate()
  n6 = VectorParameter() {
   group=10.UVDistortNoise_A
   parameterName=NoiseA_UVTransform
   channelNames {a=PannerV, b=PannerU, g=ScaleV, r=ScaleU}
   defaultValue=<rgba 1.5 0.35 -0.5 0>
  }
  n7 = AppendVector(A=@n6.ScaleU, B=@n6.ScaleU)
  n8 = TextureSampleParameter2D(UVs=Panner(Coordinate=Multiply(A=@n5, B=@n7), Speed=AppendVector(A=@n6.PannerU, B=@n6.PannerU)) {speedX=0.216}) {
   group=10.UVDistortNoise_A
   parameterName=NoiseA_Texture
   samplerType=SAMPLERTYPE_LinearGrayscale
   texture=/Game/Gadgets/Assets/VinderTech_GliderChute/Glider_Thrive_Female/FX/Texture/T_AntiMaterialCharge_Dissolve_01
  }
  n9 = VectorParameter() {
   group=10.UVDistortNoise_A
   parameterName=NoiseB_UVTransform
   channelNames {a=PannerV, b=PannerU, g=ScaleV, r=ScaleU}
   defaultValue=<rgba 0.8 0.25 0.2 0>
  }
  n10 = Panner(Coordinate=Multiply(A=TextureCoordinate(), B=AppendVector(A=@n9.ScaleU, B=@n9.ScaleU)), Speed=AppendVector(A=@n9.PannerU, B=@n9.PannerU)) {speedX=0.216}
  n11 = MaterialFunctionCall(In=Multiply(A=@n3.UGradient, B=Constant() {r=2}) {constB=8}, Contrast=ScalarParameter() {defaultValue=1, parameterName=TailGradientFallOff}) {materialFunction=&CheapContrast}
  n12 = ComponentMask(None=TextureCoordinate()) {g=true}
  n13 = OneMinus(None=@n12)
  n14 = Multiply(
   A=Power(Base=Multiply(A=@n13, B=@n12), Exp=ScalarParameter() {defaultValue=1.5, parameterName=Fade Top And Bottom Power})
   B=Clamp(None=Multiply(A=@n11.Result, B=@n4))
  )
  n15 = MaterialFunctionCall("FadeSensitivity (optional)"=ScalarParameter() {defaultValue=1, parameterName=EdgeFadeSensitivity}) {materialFunction=&FadeBasedOnViewAngle}
  n16 = ScalarParameter() {defaultValue=1500, parameterName=Fade Length}
  n17 = ScalarParameter() {defaultValue=1500, parameterName=Fade Offset}
  n18 = Multiply(
   A=Multiply(
    A=@n8.RGB
    B=TextureSampleParameter2D(UVs=@n10) {group=10.UVDistortNoise_A, parameterName=NoiseB_Texture, texture=&AtmosphericCloudNoise01}.RGB
   )
   B=Multiply(
    A=Multiply(A=@n14, B=@n15.Result)
    B=MaterialFunctionCall(Fade Length=@n16, Fade Offset=@n17) {materialFunction=&CameraDepthFade}.Result
   )
  )
  n19 = Clamp(None=Multiply(A=Multiply(A=@n1.A, B=@n18), B=ScalarParameter() {defaultValue=5, parameterName=Final Opacity Multiplier}))
  QualitySwitch(Default=DepthFade(Opacity=@n19, FadeDistance=ScalarParameter() {defaultValue=250, parameterName=Depth Fade Distance}), Low=@n19)
 ]
}
