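# Sound cue named as the traveling one-shot for wind splines in the Wind gameplay wind fields audio.
# The node graph is empty with no root, so it plays no sound as it stands.
# It carries only playback settings: the Fort_Gadgets sound class, a boosted volume, a max distance and an attenuation asset.
# That attenuation is named for the closest-point close loop rather than a one-shot.
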SoundCue /Wind/Gameplay/WindFields/Audio/WindSpline_Traveling_OneShot_Cue
root=null
nodes []
properties {
 attenuationSettings=/Wind/Gameplay/WindFields/Audio/Settings/WindSpline_ClosestPoint_Loop_Close_Att
 maxDistance=10000
 soundClassObject=&Fort_Gadgets
 volumeMultiplier=2
}
