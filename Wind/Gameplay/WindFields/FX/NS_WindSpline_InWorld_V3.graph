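# Niagara system for in-world wind field splines, showing gusts, flying debris and ground particles along a user spline range, tinted by WindColor.
# Invisible-sized audio particles travel the spline playing a wind cue, while Leader particles run its length and seed ribbon WindGusts.
# GPU debris fills the tunnel, and optional GPU ground sprites spawn near the sampled landscape before forces lift them.
# DebrisType selects sprite look and density; spline length scales spawning or lifetimes.
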NiagaraSystem /Wind/Gameplay/WindFields/FX/NS_WindSpline_InWorld_V3
dependencies {
 usedDataInterfaces [
  &NiagaraDataInterfaceArrayFloat
  &NiagaraDataInterfaceMeshRendererInfo
  &NiagaraDataInterfaceArrayColor
  &NiagaraDataInterfaceArrayInt32
  &NiagaraDataInterfaceCurve
  &NiagaraDataInterfaceAudioPlayer
  &NiagaraDataInterfaceParticleRead
 ]
 usedDynamicInputs [&Subtract_Float]
 usedModules [
  &self:GetSplineLengthSystem
  &SystemState
  &self:SystemSpawnScript.SystemScriptSource.SystemScriptGraph.NiagaraNodeAssignment_0.SetVariables_#1
  &self:SystemSpawnScript.SystemScriptSource.SystemScriptGraph.NiagaraNodeAssignment_2.SetVariables_#2
  &self:SystemSpawnScript.SystemScriptSource.SystemScriptGraph.NiagaraNodeAssignment_1.SetVariables_#3
 ]
}
emitters {
 DebrisBitsInTunnel {
  simTarget=GPUComputeSim
  emitterUpdateScript [
   &EmitterState(
    Enable Distance Culling=true
    Enable Visibility Culling=true
    MaxDistance=4000
    Module.bUseMaxDistance=true
    Module.bUseMinDistance=false
    Scalability Mode=::NewEnumerator1
   )
   &SpawnRate(
    Spawn Probability=$System.Spawn Rate Multiplier
    SpawnRate=&Multiply_Float(
     A=&ClampFloat(Float=&Divide_Float(A=$System.GetSplineLengthSystem.SplineLength, B=100), Min=150, Max=500)
     B=&SelectFloatFromArray(Direct Array Index=$User.DebrisType, Float Selection Array=DataInterface() {propertyValues {FloatData [1,0.2,0.5]}})
    )
    Use Spawn Probability=true
   )
  ]
  particleSpawnScript [
   /SetVariables(Particles.RandomInt:NiagaraInt32=&RandomRangeInt(Minimum=0, Maximum=1)) {isSetParameters=true}
   &InitializeParticle(
    Color=<rgba 1 1 1 0.5>
    Lifetime Max=4
    Lifetime Min=2
    Lifetime Mode=::NewEnumerator1
    Sprite Rotation Mode=::NewEnumerator2
    Sprite Size Mode=::NewEnumerator1
    Uniform Sprite Size Max=9
    Uniform Sprite Size Min=4
    UsePositionOffset=false
    Write Material Random=true
   ) {module=InitializeParticle001}
   /SetVariables(Particles.SpriteSize=&Multiply_Vector2D(
     A=$Particles.SpriteSize
     B=&Vector2DFromFloat(Value=&SelectFloatFromArray(Direct Array Index=$User.DebrisType, Float Selection Array=DataInterface() {propertyValues {FloatData [1,3,6]}}))
    )) {isSetParameters=true, module=SetVariables_#4}
   /SetVariables(Particles.Color=&MultiplyLinearColor(
     Linear Color=$Particles.Color
     Scale Factor=&MakeVector4_FromLinearColor(LinearColor=&SelectLinearColorFromArray(
       Array Sampling Mode=::NewEnumerator1
       Direct Array Index=$User.DebrisType
       Color Selection Array=DataInterface() {propertyValues {ColorData [<rgba 1 1 1 1>, <rgba 0.0524 0.1719 0 2>, <rgba 1 1 1 0.075>]}}
       Random Seed=?
      ))
    )) {isSetParameters=true, module=SetVariables_#5}
   /SetVariables(Particles.VisibilityTag=&SelectIntFromArray(Direct Array Index=$User.DebrisType, Int Selection Array=DataInterface() {propertyValues {IntData [0,1,2]}})) {isSetParameters=true, module=SetVariables_#6}
   /SetVariables(Particles.SpriteRotationRate=&RandomRangeFloat(Minimum=-1, Maximum=1, Evaluation Type=::NewEnumerator1)) {isSetParameters=true}
   /SetVariables() {isSetParameters=true}
   &ShapeLocation(
    Apply To Particle Position=::NewEnumerator1
    MaskSpawnGroup=false
    Offset Coordinate Space=::Local
    Offset Mode=::NewEnumerator0
    Override Seed=false
    Sphere Radius=500
    UseOwnerScale=true
   )
   /SetVariables(
    Particles.SplineEnd:NiagaraFloat=&Divide_Float(A=5000, B=$System.GetSplineLengthSystem.SplineLength)
    Particles.SplineStart:NiagaraFloat=&RandomRangeFloat(Minimum=0, Maximum=0.7)
   ) {isSetParameters=true}
   /SetVariables(Particles.SplineEnd:NiagaraFloat=&Add_Float(A=$Particles.SplineEnd, B=$Particles.SplineStart)) {isSetParameters=true}
   &SubUVAnimation(End Frame=3, SubUV Animation Mode:ENiagara_SubUVLookupMode=::NewEnumerator2)
   &DynamicMaterialParameters(
    Index 0 Param 3=1250
    Index 0 Param 4=4000
    Module.Param0WriteEnabled=false
    Module.Param1WriteEnabled=false
    Module.Param2WriteEnabled=true
    Module.Param3WriteEnabled=true
   )
  ]
  particleUpdateScript [
   &ParticleState()
   &SpriteRotationRate(Rotation Rate=&Multiply_Float(
     A=&Lerp_Float(A=-1, B=1, Alpha=&MakeFloatFromInt(Int=$Particles.RandomInt))
     B=160
    ))
   &self:Spline_Position(
    Position Along Spline:NiagaraFloat=&RemapRange(
     Input Value To Remap=&ModuloFloat(A=&Lerp_Float(A=$Particles.SplineStart, B=$Particles.SplineEnd, Alpha=$Particles.NormalizedAge), B=1)
     Input Min=0
     Input Max=1
     Output Min=$User.Start Position
     Output Max=$User.End Position
    )
    Position Offset:Vector3f=&MakeVector(
     X=&Waveform(
      Global Amplitude Scale=500
      "Waveform [1]"=::NewEnumerator0
      "[1] Phase Scale"=&RandomRangeFloat(Minimum=-1, Maximum=1)
      "[1] Frequency"=&RandomRangeFloat(Minimum=0, Maximum=1)
      "[1] Amplitude Min/Max"=<-1 1>
      "[1] Scale"=0.2
     )
     Y=&Waveform(
      Global Amplitude Scale=500
      "Waveform [1]"=::NewEnumerator0
      "[1] Phase Scale"=&RandomRangeFloat(Minimum=-1, Maximum=1)
      "[1] Frequency"=0.1
      "[1] Amplitude Min/Max"=<-1 1>
      "[1] Scale"=0.5
     )
     Z=&Waveform(
      Global Amplitude Scale=1000
      "Waveform [1]"=::NewEnumerator0
      "[1] Phase Scale"=0.2
      "[1] Frequency"=&RandomRangeFloat(Minimum=0, Maximum=1)
      "[1] Amplitude Min/Max"=<-1 1>
      "[1] Scale"=0.25
     )
    )
    Spline:NiagaraDataInterfaceSpline=$User.Spline
   ) {module=Spline_Position}
   /SetVariables(Particles.Position=&AddVectorToPosition(Position=$Output.Spline_Position.Position, Vector=$Particles.ShapeLocation.ShapeVector)) {isSetParameters=true}
   &ScaleColor(
    Scale Alpha=&FloatFromCurve(
     FloatCurve=DataInterface() {propertyValues {Curve {keys [
         {interpMode=RCIM_Cubic, tangentMode=RCTM_User, time=0, value=0}
         <at 0.2 1>
         {interpMode=RCIM_Cubic, tangentMode=RCTM_User, time=0.8, value=1}
         {interpMode=RCIM_Cubic, tangentMode=RCTM_User, time=1, value=0}
        ]}}}
     Scale Curve=3
    )
    ScaleA=true
    ScaleRGB=true
   )
   &ScaleColor(
    Scale Mode=::NewEnumerator1
    Scale RGBA=&MakeVector4_FromLinearColor(LinearColor=$User.WindColor)
    ScaleRGBA=true
   ) {module=ScaleColor001}
  ]
  rendererClasses [&NiagaraSpriteRendererProperties]
  renderers [
   {rendererClass=&NiagaraSpriteRendererProperties}
   {rendererClass=&NiagaraSpriteRendererProperties, rendererIndex=1}
   {rendererClass=&NiagaraSpriteRendererProperties, rendererIndex=2}
  ]
 }
 GroundSprites {
  simTarget=GPUComputeSim
  emitterUpdateScript [
   &EmitterState(
    Enable Distance Culling=true
    Enable Visibility Culling=true
    MaxDistance=4000
    Module.bUseMaxDistance=true
    Module.bUseMinDistance=false
    Scalability Mode=::NewEnumerator1
   )
   &SpawnRate(
    Spawn Probability=$System.Spawn Rate Multiplier
    SpawnRate=&MakeCustomFloatFromBool(
     Bool=$User.IsGroundParticlesActive
     True Float=&Multiply_Float(
      A=&ClampFloat(Float=&Divide_Float(A=$System.GetSplineLengthSystem.SplineLength, B=1), Min=250, Max=1500)
      B=&SelectFloatFromArray(Direct Array Index=$User.DebrisType, Float Selection Array=DataInterface() {propertyValues {FloatData [1,0.2,0.05]}})
     )
     False Float=0
    )
    Use Spawn Probability=true
   )
  ]
  particleSpawnScript [
   /SetVariables(Particles.RandomInt:NiagaraInt32=&RandomRangeInt(Minimum=0, Maximum=1)) {isSetParameters=true}
   &InitializeParticle(
    Color=<rgba 1 1 1 0.75>
    Lifetime Min=1.25
    Lifetime Mode=::NewEnumerator1
    Sprite Rotation Mode=::NewEnumerator2
    Sprite Size Mode=::NewEnumerator1
    Uniform Sprite Size Max=9
    Uniform Sprite Size Min=4
    UsePositionOffset=false
    Write Material Random=true
   ) {module=InitializeParticle001}
   /SetVariables(Particles.SpriteSize=&Multiply_Vector2D(
     A=$Particles.SpriteSize
     B=&Vector2DFromFloat(Value=&SelectFloatFromArray(Direct Array Index=$User.DebrisType, Float Selection Array=DataInterface() {propertyValues {FloatData [1,3,3]}}))
    )) {isSetParameters=true, module=SetVariables_#4}
   /SetVariables(Particles.Color=&MultiplyLinearColor(
     Linear Color=$Particles.Color
     Scale Factor=&MakeVector4_FromLinearColor(LinearColor=&SelectLinearColorFromArray(
       Array Sampling Mode=::NewEnumerator1
       Direct Array Index=$User.DebrisType
       Color Selection Array=DataInterface() {propertyValues {ColorData [<rgba 1 1 1 1>, <rgba 0.0524 0.1719 0 1.5>, <rgba 1 1 1 0.075>]}}
       Random Seed=?
      ))
    )) {isSetParameters=true, module=SetVariables_#5}
   /SetVariables(Particles.VisibilityTag=&SelectIntFromArray(Direct Array Index=$User.DebrisType, Int Selection Array=DataInterface() {propertyValues {IntData [0,1,2]}})) {isSetParameters=true, module=SetVariables_#6}
   /SetVariables(
    Particles.SplineEnd:NiagaraFloat=&Divide_Float(A=5000, B=$System.GetSplineLengthSystem.SplineLength)
    Particles.SplineStart:NiagaraFloat=&RandomRangeFloat(Minimum=0, Maximum=1)
   ) {isSetParameters=true}
   /SetVariables(Particles.SplineEnd:NiagaraFloat=&Add_Float(A=$Particles.SplineEnd, B=$Particles.SplineStart)) {isSetParameters=true}
   &SubUVAnimation(End Frame=3, SubUV Animation Mode:ENiagara_SubUVLookupMode=::NewEnumerator2)
   &self:Spline_Position(
    Position Along Spline:NiagaraFloat=&RemapRange(
     Input Value To Remap=&ModuloFloat(A=&Lerp_Float(A=$Particles.SplineStart, B=$Particles.SplineEnd, Alpha=$Particles.NormalizedAge), B=1)
     Input Min=0
     Input Max=1
     Output Min=$User.Start Position
     Output Max=$User.End Position
    )
    Spline:NiagaraDataInterfaceSpline=$User.Spline
   ) {module=Spline_Position}
   &ShapeLocation(
    Apply To Particle Position=::NewEnumerator1
    MaskSpawnGroup=false
    Offset Coordinate Space=::Local
    Offset Mode=::NewEnumerator0
    Override Seed=false
    Sphere Radius=1400
    UseOwnerScale=true
   )
   /SetVariables(Particles.Position=&AddVectorToPosition(Position=$Output.Spline_Position.Position, Vector=$Particles.ShapeLocation.ShapeVector)) {isSetParameters=true}
   /SetVariables(Particles.Position=&self:GetSplinePositionNearGround(
     New Landscape Sample:NiagaraDataInterfaceLandscape=$User.LandscapeDataInterface
     Random Z Offest:NiagaraFloat=&RandomRangeFloat(Minimum=0, Maximum=2)
    )) {isSetParameters=true}
   &self:Spline_Position_Offset(
    New Landscape Sample:NiagaraDataInterfaceLandscape=$User.LandscapeDataInterface
    Position Along Spline:NiagaraFloat=&Add_Float(
     A=&RemapRange(
      Input Value To Remap=&ModuloFloat(A=&Lerp_Float(A=$Particles.SplineStart, B=$Particles.SplineEnd, Alpha=$Particles.NormalizedAge), B=1)
      Input Min=0
      Input Max=1
      Output Min=$User.Start Position
      Output Max=$User.End Position
     )
     B=0.02
    )
    Spline:NiagaraDataInterfaceSpline=$User.Spline
    Z Offset=150
   ) {module=Spline_Position_Offset}
   &PointForce(
    Force Falloff Distance=500
    Force Origin=$Output.Spline_Position_Offset.SplinePos
    Force Strength=-1500
    Invert Falloff=false
    Linear Force Vector:Vector3f=<0 0 0>
    Module.Use Falloff=true
    Module.UseLinearForce=false
   )
   &ApplyInitialForces()
   &DynamicMaterialParameters(
    Index 0 Param 3=1250
    Index 0 Param 4=4000
    Module.Param0WriteEnabled=false
    Module.Param1WriteEnabled=false
    Module.Param2WriteEnabled=true
    Module.Param3WriteEnabled=true
   )
  ]
  particleUpdateScript [
   &ParticleState()
   &SpriteRotationRate(Rotation Rate=&Multiply_Float(
     A=&Lerp_Float(A=-1, B=1, Alpha=&MakeFloatFromInt(Int=$Particles.RandomInt))
     B=160
    ))
   &PointAttractionForce(
    Attraction Radius=500
    Attractor Position Offset:Vector3f=<0 0 50000>
    Attractor Position Offset Coordinate Space:ENiagaraCoordinateSpace=::Local
    Kill Within Radius=false
    Module.Use Falloff=false
    UseAttractorOffset=true
   )
   &ScaleColor(
    Scale Alpha=&FloatFromCurve(
     FloatCurve=DataInterface() {propertyValues {Curve {keys [
         {interpMode=RCIM_Cubic, tangentMode=RCTM_User, time=0, value=0}
         <at 0.2 1>
         {interpMode=RCIM_Cubic, tangentMode=RCTM_User, time=0.8, value=1}
         {interpMode=RCIM_Cubic, tangentMode=RCTM_User, time=1, value=0}
        ]}}}
     Scale Curve=3
    )
    ScaleA=true
    ScaleRGB=true
   ) {module=ScaleColor003}
   &CurlNoiseForce(
    Noise Frequency=12
    Noise Strength=&FloatFromCurve(FloatCurve=DataInterface() {propertyValues {Curve {keys [{arriveTangent=-3.2563, time=0, value=0}, <at 1 1>]}}}, Scale Curve=1500)
    Pan Noise=false
   )
   &Drag(Use Linear Drag=true, Use Rotational Drag=true)
   &GravityForce(Gravity=&Multiply_VectorByFloat(
     Vector=<0 0 -200>
     Float=&FloatFromCurve(FloatCurve=DataInterface() {propertyValues {Curve {keys [
          {arriveTangent=4.1785, interpMode=RCIM_Cubic, leaveTangent=4.1785, tangentMode=RCTM_User, time=0, value=0}
          {arriveTangent=1.1179, interpMode=RCIM_Cubic, leaveTangent=1.1179, tangentMode=RCTM_User, time=0.2042, value=0.6228}
          {arriveTangent=0.0224, interpMode=RCIM_Cubic, leaveTangent=0.0224, tangentMode=RCTM_User, time=1, value=1}
         ]}}})
    ))
   &SolveForcesAndVelocity(Clamp Velocity=false, Limit Acceleration=false)
   &ScaleSpriteSize(Uniform Curve Sprite Scale=DataInterface() {propertyValues {Curve {keys [
        <at 0 0 RCIM_Cubic>
        {arriveTangent=-0.0396, interpMode=RCIM_Cubic, leaveTangent=-0.0396, tangentMode=RCTM_User, time=0.15, value=1}
        {arriveTangent=0.0343, interpMode=RCIM_Cubic, leaveTangent=0.0343, tangentMode=RCTM_User, time=0.8, value=1}
        <at 1 0 RCIM_Cubic>
       ]}}})
   &ScaleColor(
    Scale Mode=::NewEnumerator1
    Scale RGBA=&MakeVector4_FromLinearColor(LinearColor=$User.WindColor)
    ScaleRGBA=true
   ) {module=ScaleColor001}
  ]
  rendererClasses [&NiagaraSpriteRendererProperties]
  renderers [
   {rendererClass=&NiagaraSpriteRendererProperties}
   {rendererClass=&NiagaraSpriteRendererProperties, rendererIndex=1}
   {rendererClass=&NiagaraSpriteRendererProperties, rendererIndex=2}
  ]
 }
 Leader {
  emitterUpdateScript [
   &EmitterState(
    Enable Distance Culling=true
    MaxDistance=10000
    Module.bUseMaxDistance=true
    Module.bUseMinDistance=false
    Scalability Mode=::NewEnumerator1
   )
   &SpawnRate(SpawnRate=$System.Spawn Rate Multiplier, Use Spawn Probability=false) {module=SpawnRate001}
   &SpawnRate(
    Spawn Probability=&RandomRangeFloat(Minimum=0, Maximum=0.6)
    SpawnRate=&RandomRangeFloat(Minimum=0, Maximum=1, Evaluation Type=::NewEnumerator1)
    Use Spawn Probability=true
   ) {enabled=false}
  ]
  particleSpawnScript [
   /SetVariables() {enabled=false, isSetParameters=true}
   &InitializeParticle(
    Lifetime=&Multiply_Float(
     A=&Divide_Float(A=$System.GetSplineLengthSystem.SplineLength, B=&RandomRangeFloat(Minimum=3500, Maximum=4000))
     B=0.25
    )
    Sprite Rotation Mode=::NewEnumerator2
    Sprite Size Mode=::NewEnumerator3
    Uniform Sprite Size=500
    UsePositionOffset=false
    Write Material Random=true
   )
   /SetVariables(Particles.MySplineLength:NiagaraFloat=$System.GetSplineLengthSystem.SplineLength) {isSetParameters=true}
  ]
  particleUpdateScript [
   &ParticleState()
   &ScaleSpriteSize(Uniform Curve Sprite Scale=DataInterface() {propertyValues {Curve {keys [
        {interpMode=RCIM_Cubic, tangentMode=RCTM_User, time=0, value=0}
        {interpMode=RCIM_Cubic, tangentMode=RCTM_User, time=0.5, value=1}
        {tangentMode=RCTM_User, time=1, value=0}
       ]}}})
   &self:Spline_Position(
    Position Along Spline:NiagaraFloat=&RemapRange(
     Input Value To Remap=&FloatFromCurve(FloatCurve=DataInterface() {propertyValues {Curve {keys [<at 0 0>,<at 1 1>]}}})
     Input Min=0
     Input Max=1
     Output Min=$User.Start Position
     Output Max=$User.End Position
    )
    Position Offset:Vector3f=&RandomRangeVector(Minimum=<-300 -300 -300>, Maximum=<300 300 300>)
    Spline:NiagaraDataInterfaceSpline=$User.Spline
   ) {module=Spline_Position}
   /SetVariables(Particles.Position=$Output.Spline_Position.Position) {isSetParameters=true}
   &SolveForcesAndVelocity(Clamp Velocity=false, Limit Acceleration=false)
   &CameraOffset(Camera Offset Amount=200)
  ]
  rendererClasses [&NiagaraSpriteRendererProperties]
  renderers [{rendererClass=&NiagaraSpriteRendererProperties}]
 }
 PlayLoopingAudio {
  emitterSpawnScript [/SetVariables(Emitter.LoopingAudioPlayer=DataInterface() {propertyValues {
      Attenuation=null
      Concurrency=null
      MaxPlaysPerTick=10
      SoundToPlay=/Wind/Gameplay/WindFields/Audio/WindSpline_Traveling_OneShot_Cue
      bAllowLoopingOneShotSounds=false
      bLimitPlaysPerTick=true
      bOnlyActiveDuringGameplay=true
      bStopWhenComponentIsDestroyed=true
      ConfigurationUserParameter {parameter {name=None, typeDefHandle {registeredTypeIndex=1}, varData []}}
      ParameterNames []
     }}) {isSetParameters=true}]
  emitterUpdateScript [
   &EmitterState(
    Enable Distance Culling=true
    Enable Visibility Culling=true
    MaxDistance=4000
    Module.bUseMaxDistance=true
    Module.bUseMinDistance=false
    Scalability Mode=::NewEnumerator1
   )
   &SpawnRate(SpawnRate=$System.Spawn Rate Multiplier, Use Spawn Probability=false) {module=SpawnRate001}
  ]
  particleSpawnScript [
   &InitializeParticle(
    Color Mode=::NewEnumerator0
    Lifetime Max=6
    Lifetime Min=1.5
    Lifetime Mode=::NewEnumerator1
    Sprite Size Mode=::NewEnumerator3
    Uniform Sprite Size=0
    UsePositionOffset=false
    Write Material Random=true
   )
   &ShapeLocation(
    Apply To Particle Position=::NewEnumerator1
    MaskSpawnGroup=false
    Offset Coordinate Space=::Local
    Offset Mode=::NewEnumerator0
    Override Seed=false
    Sphere Radius=400
    UseOwnerScale=true
   )
   /SetVariables(
    Particles.SplineEnd:NiagaraFloat=&Divide_Float(A=5000, B=$System.GetSplineLengthSystem.SplineLength)
    Particles.SplineStart:NiagaraFloat=&RandomRangeFloat(Minimum=0, Maximum=1)
   ) {isSetParameters=true}
   /SetVariables(Particles.SplineEnd:NiagaraFloat=&Add_Float(A=$Particles.SplineEnd, B=$Particles.SplineStart)) {isSetParameters=true}
  ]
  particleUpdateScript [
   &ParticleState()
   &self:Spline_Position(
    Position Along Spline:NiagaraFloat=&RemapRange(
     Input Value To Remap=&ModuloFloat(A=&Lerp_Float(A=$Particles.SplineStart, B=$Particles.SplineEnd, Alpha=$Particles.NormalizedAge), B=1)
     Input Min=0
     Input Max=1
     Output Min=$User.Start Position
     Output Max=$User.End Position
    )
    Spline:NiagaraDataInterfaceSpline=$User.Spline
   ) {module=Spline_Position}
   /SetVariables(Particles.Position=&AddVectorToPosition(Position=$Output.Spline_Position.Position, Vector=$Particles.ShapeLocation.ShapeVector)) {isSetParameters=true}
   /Niagara/Modules/Audio/PlayPersistentAudio(
    Audio Player:NiagaraDataInterfaceAudioPlayer=$Emitter.LoopingAudioPlayer
    Enable Fade In / Out=true
    Fade In Time=2
    Fade Out Time=2
    Play Audio:NiagaraBool=$DataInstance.Alive
   )
   /Niagara/Modules/Audio/UpdatePersistentAudio(Audio Player:NiagaraDataInterfaceAudioPlayer=$Emitter.LoopingAudioPlayer, Update Type:ENiagara_AudioParamType=::NewEnumerator5)
  ]
  rendererClasses []
  renderers []
 }
 WindGusts {
  emitterUpdateScript [
   &EmitterState(
    Enable Distance Culling=true
    MaxDistance=10000
    Module.bUseMaxDistance=true
    Module.bUseMinDistance=false
    Scalability Mode=::NewEnumerator1
   )
   &SpawnParticlesFromOtherEmitter(
    Attribute Reader=DataInterface() {propertyValues {EmitterBinding {bindingMode=Other, emitterName=Leader}}}
    Spawn Rate=&Multiply_Float(A=7, B=&Multiply_Float(A=$User.WindGustSpawnScale, B=1))
    Use Maximum Age=false
    Use Minimum Age=false
   )
  ]
  particleSpawnScript [
   &InitializeParticle(
    Color=<rgba 6 6 6 1>
    Lifetime=&ClampFloat(Float=&Divide_Float(A=$System.GetSplineLengthSystem.SplineLength, B=5000), Min=2, Max=3)
    Ribbon Width=60
    Ribbon Width Mode=::NewEnumerator1
    UsePositionOffset=false
    Write Material Random=true
   )
   &SampleParticlesFromOtherEmitter(Particle ID Sampling=::NewEnumerator1)
  ]
  particleUpdateScript [
   &ParticleState()
   &CurlNoiseForce(Noise Frequency=0.6, Noise Strength=75, Pan Noise=false)
   &SolveForcesAndVelocity(Clamp Velocity=false, Limit Acceleration=false)
   &ScaleRibbonWidth(Ribbon Width Scale=&FloatFromCurve(
     FloatCurve=DataInterface() {propertyValues {Curve {keys [
         {interpMode=RCIM_Cubic, tangentMode=RCTM_User, time=0, value=0}
         {interpMode=RCIM_Cubic, tangentMode=RCTM_User, time=0.5, value=1}
         {tangentMode=RCTM_User, time=1, value=0}
        ]}}}
     Scale Curve=3
    ))
   &ScaleColor(Scale Mode=1, Scale RGBA=&MakeVector4_FromLinearColor(LinearColor=$User.WindColor), ScaleRGBA=true) {module=ScaleColor001}
   &ScaleColor(Scale Alpha=100, Scale RGB=<100 0 0>, ScaleA=true, ScaleRGB=true) {enabled=false}
  ]
  rendererClasses [&NiagaraRibbonRendererProperties, &NiagaraSpriteRendererProperties]
  renderers [
   {rendererClass=&NiagaraRibbonRendererProperties}
   {rendererClass=&NiagaraSpriteRendererProperties, rendererIndex=1}
  ]
 }
}
summary {
 systemName=NS_WindSpline_InWorld_V3
 emitters [
  {emitterName=DebrisBitsInTunnel, simTarget=GPUComputeSim, rendererClasses [&NiagaraSpriteRendererProperties]}
  {emitterName=PlayLoopingAudio, rendererClasses []}
  {emitterName=Leader, rendererClasses [&NiagaraSpriteRendererProperties]}
  {emitterName=WindGusts, rendererClasses [&NiagaraRibbonRendererProperties, &NiagaraSpriteRendererProperties]}
  {emitterName=GroundSprites, simTarget=GPUComputeSim, rendererClasses [&NiagaraSpriteRendererProperties]}
 ]
 userVariables [
  User.DebrisType {type=NiagaraInt32, defaultValue=2}
  User.End Position {type=NiagaraFloat, defaultValue=1, description=""}
  User.IsGroundParticlesActive {type=NiagaraBool, defaultValue=true}
  User.LandscapeDataInterface {
   type=NiagaraDataInterfaceLandscape
   description=""
   defaultValue {
    struct=NiagaraExt_VariableValue_DataInterface
    value {
     dataInterface=&self:NiagaraDataInterfaceLandscape_0
     dataInterfaceClass=&NiagaraDataInterfaceLandscape
    }
   }
  }
  User.Spline {
   type=NiagaraDataInterfaceSpline
   description=""
   defaultValue {
    struct=NiagaraExt_VariableValue_DataInterface
    value {
     dataInterface=&self:NiagaraDataInterfaceSpline_1
     dataInterfaceClass=/Script/Niagara.NiagaraDataInterfaceSpline
    }
   }
  }
  User.Start Position {type=NiagaraFloat, defaultValue=0, description=""}
  User.WindColor {type=LinearColor, description="", defaultValue {struct=LinearColor, value=<rgba 1 1 1 1>}}
  User.WindGustSpawnScale {type=NiagaraFloat, defaultValue=1, description=""}
 ]
}
