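# Niagara system for wind-field world interaction whose only enabled emitter blows soft, faintly blue-tinted sprites along X from a thin box.
# Those sprites grow, sink under downward acceleration and drag, and fade out.
# Spawn probability falls with distance, and the emitter culls at range.
# Three disabled GPU emitters hold spline-following variants using curl noise and waveform offsets or landscape sampling, so the spline user parameters drive nothing enabled.
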NiagaraSystem /Wind/Gameplay/WindFields/FX/NS_WindSpline_WorldInteraction_Wind
dependencies {
 usedDataInterfaces [
  &NiagaraDataInterfaceMeshRendererInfo
  &NiagaraDataInterfaceCollisionQuery
  &NiagaraDataInterfaceCurve
  &NiagaraDataInterfaceColorCurve
 ]
 usedDynamicInputs [&Subtract_Float]
 usedModules [
  &SystemState
  &self:SystemSpawnScript.SystemScriptSource.SystemScriptGraph.NiagaraNodeAssignment_0.SetVariables_#1
  &self:SystemSpawnScript.SystemScriptSource.SystemScriptGraph.NiagaraNodeAssignment_1.SetVariables_#2
 ]
}
emitters {
 FN_Sprite {
  bEnabled=false
  simTarget=GPUComputeSim
  emitterUpdateScript [
   &EmitterState(
    Enable Distance Culling=true
    MaxDistance=150000
    Module.bUseMaxDistance=true
    Module.bUseMinDistance=false
    Scalability Mode=::NewEnumerator1
   )
   &self:GetSplineLength() {module=GetSplineLength001}
   &SpawnRate(
    SpawnRate=&ClampFloat(Float=&Divide_Float(A=$Output.GetSplineLength001.SplineLength, B=1), Min=100, Max=120)
    Use Spawn Probability=false
   )
  ]
  particleSpawnScript [
   &InitializeParticle(
    Color=&RandomRangeLinearColor(Minimum=<rgba 1 1 1 1>, Maximum=<rgba 3 3 3 1>)
    Lifetime Min=1.5
    Lifetime Mode=::NewEnumerator1
    Sprite Rotation Angle=&RandomRangeFloat(Minimum=140, Maximum=220)
    Sprite Rotation Mode=::NewEnumerator0
    Sprite Size Mode=::NewEnumerator1
    Uniform Sprite Size Max=3500
    Uniform Sprite Size Min=2500
    UsePositionOffset=false
    Write Material Random=true
   ) {module=InitializeParticle001}
   /SetVariables(Particles.SpriteRotationRate=&RandomRangeFloat(Minimum=-0.1, Maximum=0.1, Evaluation Type=::NewEnumerator1)) {isSetParameters=true}
   /SetVariables() {isSetParameters=true, module=SetVariables_#3}
   &ShapeLocation(
    Apply To Particle Position=::NewEnumerator1
    MaskSpawnGroup=false
    Non Uniform Scale=<1 1 0.1>
    Offset Coordinate Space=::Local
    Offset Mode=::NewEnumerator0
    Override Seed=false
    Sphere Radius=500
    UseOwnerScale=true
   )
   &AddVelocity(Velocity=<0 0 2500>)
   &self:GetSplineLength() {module=GetSplineLength}
   /SetVariables(
    Particles.SplineEnd:NiagaraFloat=&Divide_Float(A=3000, B=$Output.GetSplineLength.SplineLength)
    Particles.SplineStart:NiagaraFloat=&RandomRangeFloat(Minimum=0, Maximum=0.7)
   ) {isSetParameters=true, module=SetVariables_#4}
   /SetVariables(Particles.SplineEnd:NiagaraFloat=&Add_Float(A=$Particles.SplineEnd, B=$Particles.SplineStart)) {isSetParameters=true, module=SetVariables_#5}
   /Niagara/Modules/Math/MoveToNearestDistanceFieldSurface_GPU(Kill by Distance=25000, bKillByDistance=true) {enabled=false}
   /SetVariables(Particles.InitialPos:NiagaraPosition=$Particles.Position) {enabled=false, isSetParameters=true, module=SetVariables_#6}
   /Niagara/Modules/Math/PlaceParticlesOnDepthBuffer_GPU(Max Z Depth=15000, Module.Use Max Z Depth=true) {enabled=false}
   &KillParticles(Kill Particles=&SetBoolByFloatComparison(
     A=&VectorLength(Vector=&Subtract_Vector(A=$Particles.InitialPos, B=Linked() {linkedVariable {name=Particles.Position, type=NiagaraPosition}}))
     B=15000
    )) {enabled=false}
   &KillParticles(Kill Particles=&Boolean_NotOperation(Bool=$Output.PlaceParticlesOnDepthBuffer_GPU.ValidSample)) {enabled=false, module=KillParticles001}
  ]
  particleUpdateScript [
   &ParticleState()
   &CurlNoiseForce(
    Noise Frequency=&RandomRangeFloat(Minimum=0.25, Maximum=2)
    Noise Strength=&RandomRangeFloat(Minimum=2500, Maximum=2500)
    Pan Noise=false
   )
   &WindForce(Air Resistance=0, Wind Speed=<250 0 0>)
   &PointAttractionForce(
    Attraction Radius=1
    Attractor Position Offset:Vector3f=<0 0 50000>
    Kill Within Radius=false
    Module.Use Falloff=false
    UseAttractorOffset=true
   ) {enabled=false}
   &SolveForcesAndVelocity(Clamp Velocity=true, Limit Acceleration=true, Speed Limit=2500)
   /SetVariables(Particles.New Position:NiagaraPosition=$Particles.Position) {isSetParameters=true, module=SetVariables_#7}
   &self:Spline_Position(
    Position Along Spline:NiagaraFloat=&Add_Float(
     A=&FloatFromCurve(
      FloatCurve=DataInterface() {propertyValues {Curve {keys [{tangentMode=RCTM_User, time=0, value=0}, {tangentMode=RCTM_User, time=1, value=1}]}}}
      CurveIndex=&Lerp_Float(A=$Particles.SplineStart, B=$Particles.SplineEnd, Alpha=$Particles.NormalizedAge)
     )
     B=&RandomRangeFloat(Minimum=0, Maximum=0)
    )
    Position Offset:Vector3f=&MakeVector(
     X=&Waveform(
      Global Amplitude Scale=500
      "Waveform [1]"=::NewEnumerator0
      "[1] Phase Scale"=&RandomRangeFloat(Minimum=-1, Maximum=1)
      "[1] Frequency"=&RandomRangeFloat(Minimum=0, Maximum=1)
      "[1] Amplitude Min/Max"=<-1 1>
      "[1] Scale"=0.2
     )
     Y=&Waveform(
      Global Amplitude Scale=500
      "Waveform [1]"=::NewEnumerator0
      "[1] Phase Scale"=&RandomRangeFloat(Minimum=-1, Maximum=1)
      "[1] Frequency"=0.1
      "[1] Amplitude Min/Max"=<-1 1>
      "[1] Scale"=0.5
     )
     Z=&Waveform(
      Global Amplitude Scale=1000
      "Waveform [1]"=::NewEnumerator0
      "[1] Phase Scale"=0.5
      "[1] Frequency"=&RandomRangeFloat(Minimum=0, Maximum=1)
      "[1] Amplitude Min/Max"=<-1 1>
      "[1] Scale"=0.25
     )
    )
    Spline:NiagaraDataInterfaceSpline=$User.Spline
   ) {module=Spline_Position}
   &DynamicMaterialParameters(
    Index 0 Param 1=&FloatFromCurve(FloatCurve=DataInterface() {propertyValues {Curve {keys [{tangentMode=RCTM_User, time=0, value=0}, {tangentMode=RCTM_User, time=1, value=0.5}]}}})
    Index 0 Param 2=&RandomRangeFloat(Minimum=1, Maximum=1)
    Index 0 Param 3=1
    Index 1 Param 1=500
    Index 1 Param 2=1500
    Module.Param0WriteEnabled=true
    Module.Param0WriteEnabled1=true
    Module.Param1WriteEnabled=true
    Module.Param1WriteEnabled1=true
    Module.Param2WriteEnabled=true
    Module.Param2WriteEnabled1=true
    Module.Param3WriteEnabled=true
    Module.Param3WriteEnabled1=true
    Write Parameter Index 1=true
   ) {module=DynamicMaterialParameters001}
   &ScaleColor(
    Scale Alpha=&FloatFromCurve(
     FloatCurve=DataInterface() {propertyValues {Curve {keys [
         {interpMode=RCIM_Cubic, tangentMode=RCTM_User, time=0, value=0}
         <at 0.2 1>
         {interpMode=RCIM_Cubic, tangentMode=RCTM_User, time=0.8, value=1}
         {interpMode=RCIM_Cubic, tangentMode=RCTM_User, time=1, value=0}
        ]}}}
     Scale Curve=&Multiply_Float(
      A=0.1
      B=&FN_LOD_Distance_Remap(Start Distance=500, End Distance=25000, Minimum Output=1, Maximum Output=2)
     )
    )
    ScaleA=true
    ScaleRGB=true
   )
   &SpriteRotationRate(Delta Time:NiagaraFloat=$Particles.NormalizedAge, Rotation Rate=$Particles.SpriteRotationRate)
   &ScaleSpriteSize(
    Uniform Curve Scale=&Multiply_Float(
     A=1
     B=&FN_LOD_Distance_Remap(Start Distance=5000, End Distance=15000, Minimum Output=0.5, Maximum Output=0.7)
    )
    Uniform Curve Sprite Scale=DataInterface() {propertyValues {Curve {keys [
        {arriveTangent=1.4932, interpMode=RCIM_Cubic, leaveTangent=1.4932, tangentMode=RCTM_User, time=0, value=0.5}
        {arriveTangent=-0.0405, interpMode=RCIM_Cubic, leaveTangent=-0.0405, tangentMode=RCTM_User, time=1, value=1}
       ]}}}
   )
   /SetVariables(Particles.Position=&Multiply_Vector(
     A=&AddVectorToPosition(
      Position=/Niagara/DynamicInputs/Lerp/Lerp_VectorComponents(
       Lerp Component Alpha:Vector3f=<0.5 0.5 0>
       A:Vector3f=Linked() {linkedVariable {name=Output.Spline_Position.Position, type=NiagaraPosition}}
       B:Vector3f=Linked() {linkedVariable {name=Particles.New Position, type=NiagaraPosition}}
      )
      Vector=$Particles.ShapeLocation.ShapeVector
     )
     B=<1 1 0.56>
    )) {isSetParameters=true, module=SetVariables_#8}
   &CameraOffset(Camera Offset Amount=500)
   &SubUVAnimation(SubUV Loop Count=1.5)
  ]
  rendererClasses [&NiagaraSpriteRendererProperties]
  renderers [{rendererClass=&NiagaraSpriteRendererProperties}]
 }
 FN_Sprite001 {
  bEnabled=false
  simTarget=GPUComputeSim
  emitterUpdateScript [
   &EmitterState(
    Enable Distance Culling=true
    MaxDistance=150000
    Module.bUseMaxDistance=true
    Module.bUseMinDistance=false
    Scalability Mode=::NewEnumerator1
   )
   &self:GetSplineLength() {module=GetSplineLength001}
   &SpawnRate(
    SpawnRate=&MakeCustomFloatFromBool(
     Bool=$User.IsActive
     True Float=&ClampFloat(Float=&Divide_Float(A=$Output.GetSplineLength001.SplineLength, B=1), Min=250, Max=1500)
     False Float=0
    )
    Use Spawn Probability=false
   )
  ]
  particleSpawnScript [
   &InitializeParticle(
    Color=<rgba 1 1 1 2>
    Lifetime Min=1.25
    Lifetime Mode=::NewEnumerator1
    Sprite Rotation Angle=&RandomRangeFloat(Minimum=0, Maximum=360)
    Sprite Rotation Mode=::NewEnumerator0
    Sprite Size Mode=::NewEnumerator1
    Uniform Sprite Size Max=6
    Uniform Sprite Size Min=2
    UsePositionOffset=false
    Write Material Random=true
   ) {module=InitializeParticle001}
   /SetVariables() {isSetParameters=true, module=SetVariables_#3}
   /SetVariables(Particles.MaxPosition:NiagaraFloat=$User.Progress) {isSetParameters=true}
   /SetVariables(Particles.MinPosition:NiagaraFloat=$User.MinPosition) {isSetParameters=true}
   &self:GetSplineLength() {module=GetSplineLength}
   /SetVariables(
    Particles.SplineEnd:NiagaraFloat=&Divide_Float(A=800, B=$Output.GetSplineLength.SplineLength)
    Particles.SplineStart:NiagaraFloat=&RandomRangeFloat(Minimum=0, Maximum=1)
   ) {isSetParameters=true, module=SetVariables_#4}
   /SetVariables(Particles.SplineEnd:NiagaraFloat=&Add_Float(A=$Particles.SplineEnd, B=$Particles.SplineStart)) {isSetParameters=true, module=SetVariables_#5}
   &SubUVAnimation(End Frame=3, SubUV Animation Mode:ENiagara_SubUVLookupMode=::NewEnumerator2)
   &self:Spline_Position_01(
    Position Along Spline:NiagaraFloat=&FloatFromCurve(
     FloatCurve=DataInterface() {propertyValues {Curve {keys [{tangentMode=RCTM_User, time=0, value=0}, {tangentMode=RCTM_User, time=1, value=1}]}}}
     CurveIndex=&Lerp_Float(A=$Particles.SplineStart, B=$Particles.SplineEnd, Alpha=$Particles.NormalizedAge)
    )
    Spline:NiagaraDataInterfaceSpline=$User.Spline
   ) {module=Spline_Position_01}
   &ShapeLocation(
    Apply To Particle Position=::NewEnumerator1
    MaskSpawnGroup=false
    Offset Coordinate Space=::Local
    Offset Mode=::NewEnumerator0
    Override Seed=false
    Sphere Radius=1400
    UseOwnerScale=true
   )
   /SetVariables(Particles.Position=&AddVectorToPosition(Position=$Output.Spline_Position_01.Position, Vector=$Particles.ShapeLocation.ShapeVector)) {isSetParameters=true}
   /SetVariables(Particles.Position=&self:ScratchDynamicInput_01(
     New Landscape Sample:NiagaraDataInterfaceLandscape=$User.LandscapeDataInterface
     Random Z Offest:NiagaraFloat=&RandomRangeFloat(Minimum=0, Maximum=0)
    )) {isSetParameters=true}
   &self:Spline_Position_Offset(
    Position Along Spline:NiagaraFloat=&Add_Float(
     A=&FloatFromCurve(
      FloatCurve=DataInterface() {propertyValues {Curve {keys [{tangentMode=RCTM_User, time=0, value=0}, {tangentMode=RCTM_User, time=1, value=1}]}}}
      CurveIndex=&Lerp_Float(A=$Particles.SplineStart, B=$Particles.SplineEnd, Alpha=$Particles.NormalizedAge)
     )
     B=0.02
    )
    Spline:NiagaraDataInterfaceSpline=$User.Spline
    Z Offset=150
   ) {module=Spline_Position_Offset}
   &PointForce(
    Force Falloff Distance=500
    Force Origin=$Output.Spline_Position_Offset.SplinePos
    Force Strength=-1500
    Invert Falloff=false
    Linear Force Vector:Vector3f=<0 0 0>
    Module.Use Falloff=true
    Module.UseLinearForce=false
   )
   &ApplyInitialForces()
  ]
  particleUpdateScript [
   &ParticleState()
   &PointAttractionForce(
    Attraction Radius=500
    Attraction Strength=5
    Attractor Position Offset:Vector3f=<0 0 50000>
    Attractor Position Offset Coordinate Space:ENiagaraCoordinateSpace=::Local
    Kill Within Radius=false
    Module.Use Falloff=false
    UseAttractorOffset=true
   )
   &ScaleColor(
    Scale Alpha=&FloatFromCurve(
     FloatCurve=DataInterface() {propertyValues {Curve {keys [
         {interpMode=RCIM_Cubic, tangentMode=RCTM_User, time=0, value=0}
         <at 0.2 1>
         {interpMode=RCIM_Cubic, tangentMode=RCTM_User, time=0.8, value=1}
         {interpMode=RCIM_Cubic, tangentMode=RCTM_User, time=1, value=0}
        ]}}}
     Scale Curve=&Multiply_Float(
      A=0.1
      B=&FN_LOD_Distance_Remap(Start Distance=500, End Distance=25000, Minimum Output=2, Maximum Output=8)
     )
    )
    ScaleA=true
    ScaleRGB=true
   ) {module=ScaleColor003}
   &CameraOffset(Camera Offset Amount=200)
   &CurlNoiseForce(
    Noise Frequency=12
    Noise Strength=&FloatFromCurve(FloatCurve=DataInterface() {propertyValues {Curve {keys [{arriveTangent=-3.2563, time=0, value=0}, <at 1 1>]}}}, Scale Curve=1500)
    Pan Noise=false
   )
   &Drag(Use Linear Drag=true, Use Rotational Drag=true)
   &SolveForcesAndVelocity(Clamp Velocity=false, Limit Acceleration=false)
  ]
  rendererClasses [&NiagaraSpriteRendererProperties]
  renderers [{rendererClass=&NiagaraSpriteRendererProperties}]
 }
 FN_Sprite002 {
  bEnabled=false
  simTarget=GPUComputeSim
  emitterUpdateScript [
   &EmitterState(
    Enable Distance Culling=true
    MaxDistance=150000
    Module.bUseMaxDistance=true
    Module.bUseMinDistance=false
    Scalability Mode=::NewEnumerator1
   )
   &self:GetSplineLength() {module=GetSplineLength001}
   &SpawnRate(
    SpawnRate=&ClampFloat(Float=&Divide_Float(A=$Output.GetSplineLength001.SplineLength, B=1), Min=100, Max=100)
    Use Spawn Probability=false
   )
  ]
  particleSpawnScript [
   &InitializeParticle(
    Color=&RandomRangeLinearColor(Minimum=<rgba 1 1 1 1>, Maximum=<rgba 3 3 3 1>)
    Lifetime Mode=::NewEnumerator1
    Mesh Scale=<1 5 5>
    Mesh Scale Mode=::NewEnumerator4
    Sprite Rotation Angle=&RandomRangeFloat(Minimum=0, Maximum=360)
    Sprite Rotation Mode=::NewEnumerator0
    Sprite Size Mode=::NewEnumerator1
    Uniform Sprite Size Max=65
    Uniform Sprite Size Min=25
    UsePositionOffset=false
    Write Material Random=true
   ) {module=InitializeParticle001}
   /SetVariables() {isSetParameters=true, module=SetVariables_#3}
   &ShapeLocation(
    Apply To Particle Position=::NewEnumerator1
    MaskSpawnGroup=false
    Offset Coordinate Space=::Local
    Offset Mode=::NewEnumerator0
    Override Seed=false
    Sphere Radius=1
    UseOwnerScale=true
   )
   &self:GetSplineLength() {module=GetSplineLength}
   &AddVelocity(Velocity=<0 5000 5000>)
   /SetVariables(
    Particles.SplineEnd:NiagaraFloat=&Divide_Float(A=2000, B=$Output.GetSplineLength.SplineLength)
    Particles.SplineStart:NiagaraFloat=&RandomRangeFloat(Minimum=0, Maximum=0.8)
   ) {isSetParameters=true, module=SetVariables_#4}
   /SetVariables(Particles.SplineEnd:NiagaraFloat=&Add_Float(A=$Particles.SplineEnd, B=$Particles.SplineStart)) {isSetParameters=true, module=SetVariables_#5}
   &MoveToNearestDistanceFieldSurface_GPU(Kill by Distance=25000, bKillByDistance=true) {enabled=false}
   /SetVariables(Particles.InitialPos:NiagaraPosition=$Particles.Position) {enabled=false, isSetParameters=true, module=SetVariables_#6}
   &PlaceParticlesOnDepthBuffer_GPU(Max Z Depth=15000, Module.Use Max Z Depth=true) {enabled=false}
   &KillParticles(Kill Particles=&SetBoolByFloatComparison(
     A=&VectorLength(Vector=&Subtract_Vector(A=$Particles.InitialPos, B=Linked() {linkedVariable {name=Particles.Position, type=NiagaraPosition}}))
     B=15000
    )) {enabled=false}
   &KillParticles(Kill Particles=&Boolean_NotOperation(Bool=$Output.PlaceParticlesOnDepthBuffer_GPU.ValidSample)) {enabled=false, module=KillParticles001}
   &SubUVAnimation(End Frame=3, SubUV Animation Mode:ENiagara_SubUVLookupMode=::NewEnumerator2)
   &InitialMeshOrientation(Mesh Axis To Orient=<0 0 1>, Mesh Orientation Mode=::NewEnumerator2, Orientation Vector=<0 1 0>, Use Rotation Vector=false)
   &CameraOffset() {module=CameraOffset001}
  ]
  particleUpdateScript [
   &ParticleState()
   &CurlNoiseForce(
    Noise Frequency=&RandomRangeFloat(Minimum=0.25, Maximum=2)
    Noise Strength=&RandomRangeFloat(Minimum=5, Maximum=5)
    Pan Noise=false
   )
   &WindForce(Air Resistance=0, Wind Speed=<5000 0 0>)
   &PointAttractionForce(
    Attraction Radius=500000
    Attraction Strength=5
    Attractor Position Offset:Vector3f=<0 0 50000>
    Attractor Position Offset Coordinate Space:ENiagaraCoordinateSpace=::Local
    Kill Within Radius=false
    Module.Use Falloff=false
    UseAttractorOffset=true
   ) {enabled=false}
   &SolveForcesAndVelocity(Clamp Velocity=false, Limit Acceleration=false, Speed Limit=5000)
   /SetVariables(Particles.New Position:NiagaraPosition=$Particles.Position) {isSetParameters=true, module=SetVariables_#7}
   &self:Spline_Position(
    Position Along Spline:NiagaraFloat=&Add_Float(
     A=&FloatFromCurve(
      FloatCurve=DataInterface() {propertyValues {Curve {keys [{tangentMode=RCTM_User, time=0, value=0}, {tangentMode=RCTM_User, time=1, value=1}]}}}
      CurveIndex=&Lerp_Float(A=$Particles.SplineStart, B=$Particles.SplineEnd, Alpha=$Particles.NormalizedAge)
     )
     B=&RandomRangeFloat(Minimum=0, Maximum=0)
    )
    Position Offset:Vector3f=&MakeVector(
     X=&Waveform(
      Global Amplitude Scale=500
      "Waveform [1]"=::NewEnumerator0
      "[1] Phase Scale"=&RandomRangeFloat(Minimum=-1, Maximum=1)
      "[1] Frequency"=&RandomRangeFloat(Minimum=0, Maximum=1)
      "[1] Amplitude Min/Max"=<-1 1>
      "[1] Scale"=0.2
     )
     Y=&Waveform(
      Global Amplitude Scale=500
      "Waveform [1]"=::NewEnumerator0
      "[1] Phase Scale"=&RandomRangeFloat(Minimum=-1, Maximum=1)
      "[1] Frequency"=0.1
      "[1] Amplitude Min/Max"=<-1 1>
      "[1] Scale"=0.5
     )
     Z=&Waveform(
      Global Amplitude Scale=1000
      "Waveform [1]"=::NewEnumerator0
      "[1] Frequency"=&RandomRangeFloat(Minimum=0, Maximum=1)
      "[1] Amplitude Min/Max"=<-1 1>
      "[1] Scale"=0.25
     )
    )
    Spline:NiagaraDataInterfaceSpline=$User.Spline
   ) {module=Spline_Position}
   &ScaleColor(
    Scale Alpha=&FloatFromCurve(
     FloatCurve=DataInterface() {propertyValues {Curve {keys [
         {interpMode=RCIM_Cubic, tangentMode=RCTM_User, time=0, value=0}
         <at 0.2 1>
         {interpMode=RCIM_Cubic, tangentMode=RCTM_User, time=0.8, value=1}
         {interpMode=RCIM_Cubic, tangentMode=RCTM_User, time=1, value=0}
        ]}}}
     Scale Curve=&Multiply_Float(
      A=0.1
      B=&FN_LOD_Distance_Remap(Start Distance=500, End Distance=25000, Minimum Output=25, Maximum Output=25)
     )
    )
    Scale RGB=<5 5 5>
    ScaleA=true
    ScaleRGB=true
   )
   &ScaleSpriteSize(
    Uniform Curve Scale=&Multiply_Float(
     A=1
     B=&FN_LOD_Distance_Remap(Start Distance=5000, End Distance=15000, Minimum Output=0.5, Maximum Output=0.7)
    )
    Uniform Curve Sprite Scale=DataInterface() {propertyValues {Curve {keys [
        {arriveTangent=2.5316, interpMode=RCIM_Cubic, leaveTangent=2.5316, tangentMode=RCTM_User, time=0, value=0}
        {arriveTangent=-0.0405, interpMode=RCIM_Cubic, leaveTangent=-0.0405, tangentMode=RCTM_User, time=1, value=1}
       ]}}}
   )
   /SetVariables(Particles.Position=&Multiply_Vector(
     A=&AddVectorToPosition(
      Position=&Lerp_VectorComponents(
       Lerp Component Alpha:Vector3f=<0.5 0.5 0>
       A:Vector3f=Linked() {linkedVariable {name=Output.Spline_Position.Position, type=NiagaraPosition}}
       B:Vector3f=Linked() {linkedVariable {name=Particles.New Position, type=NiagaraPosition}}
      )
      Vector=$Particles.ShapeLocation.ShapeVector
     )
     B=<1 1 0.56>
    )) {isSetParameters=true, module=SetVariables_#8}
   &CameraOffset(Camera Offset Amount=500)
   &DynamicMaterialParameters(
    Index 0 Param 3=1
    Index 0 Param 4=1
    Module.Param0WriteEnabled=true
    Module.Param1WriteEnabled=true
    Module.Param2WriteEnabled=true
    Module.Param3WriteEnabled=true
   )
  ]
  rendererClasses [&NiagaraSpriteRendererProperties, &NiagaraMeshRendererProperties]
  renderers [
   {rendererClass=&NiagaraSpriteRendererProperties}
   {rendererClass=&NiagaraMeshRendererProperties, rendererIndex=1}
  ]
 }
 FN_Sprite003 {
  emitterUpdateScript [
   &EmitterState(Enable Distance Culling=true, MaxDistance=8000, Module.bUseMaxDistance=true, Module.bUseMinDistance=false, Scalability Mode=::NewEnumerator1)
   &SpawnRate(SpawnRate=14, Use Spawn Probability=false) {enabled=false, module=SpawnRate001}
   &SpawnRate(
    Spawn Probability=&ClampFloat(Float=&FN_LOD_Distance_Remap(Start Distance=1500, End Distance=8000, Minimum Output=1, Maximum Output=0))
    SpawnRate=14
    Use Spawn Probability=true
   )
  ]
  particleSpawnScript [
   &InitializeParticle(
    Color=<rgba 0.7332 0.8281 1 0.2>
    Lifetime Max=2.5
    Lifetime Min=1.5
    Lifetime Mode=::NewEnumerator1
    Sprite Rotation Angle=&RandomRangeFloat(Minimum=0, Maximum=360)
    Sprite Rotation Mode=::NewEnumerator0
    Sprite Size Mode=::NewEnumerator1
    Uniform Sprite Size Max=150
    Uniform Sprite Size Min=100
    UsePositionOffset=true
    Write Material Random=true
   ) {module=InitializeParticle001}
   &ShapeLocation(
    Box Size=<0 800 600>
    MaskSpawnGroup=false
    Override Seed=false
    Rotation Mode=::NewEnumerator1
    Shape Primitive=::NewEnumerator2
    Surface Only=false
    UseOwnerScale=true
   )
   &AddVelocity(
    Velocity=&MakeVector(X=&RandomRangeFloat(Minimum=600, Maximum=1200), Y=0, Z=0)
    Velocity Speed Scale=1.5
   )
   &CameraOffset(Camera Offset Amount=200)
   &DynamicMaterialParameters(
    Index 0 Param 1=1
    Index 0 Param 2=0.4
    Index 0 Param 3=1
    Index 0 Param 4=1
    Module.Param0WriteEnabled=true
    Module.Param1WriteEnabled=true
    Module.Param2WriteEnabled=true
    Module.Param3WriteEnabled=true
   )
  ]
  particleUpdateScript [
   &ParticleState()
   &ScaleSpriteSize(
    Uniform Curve Scale=5
    Uniform Curve Sprite Scale=DataInterface() {propertyValues {Curve {keys [
        {arriveTangent=4.1785, interpMode=RCIM_Cubic, leaveTangent=4.1785, tangentMode=RCTM_User, time=0, value=0.5}
        {arriveTangent=1.1179, interpMode=RCIM_Cubic, leaveTangent=1.1179, tangentMode=RCTM_User, time=0.2, value=0.825}
        {arriveTangent=0.0224, interpMode=RCIM_Cubic, leaveTangent=0.0224, tangentMode=RCTM_User, time=1, value=1}
       ]}}}
   )
   &SubUVAnimation(Start Frame=1, SubUV Animation Mode:ENiagara_SubUVLookupMode=::NewEnumerator1)
   &AccelerationForce(Acceleration=&MakeVector(X=0, Y=0, Z=-500))
   &Drag(Drag=1.2, Use Linear Drag=true, Use Rotational Drag=true)
   &SolveForcesAndVelocity(Clamp Velocity=false, Limit Acceleration=false)
   &ScaleColor(
    Scale Alpha=&FloatFromCurve(FloatCurve=DataInterface() {propertyValues {Curve {keys [
         {interpMode=RCIM_Cubic, tangentMode=RCTM_User, time=0, value=0}
         {interpMode=RCIM_Cubic, leaveTangent=-2.3643, tangentMode=RCTM_Break, time=0.2, value=0.95}
         {arriveTangent=-1.2778, interpMode=RCIM_Cubic, leaveTangent=-1.2778, tangentMode=RCTM_User, time=1, value=0}
        ]}}})
    ScaleA=true
    ScaleRGB=true
   )
  ]
  rendererClasses [&NiagaraSpriteRendererProperties]
  renderers [{rendererClass=&NiagaraSpriteRendererProperties}]
 }
}
summary {
 systemName=NS_WindSpline_WorldInteraction_Wind
 emitters [
  {bEnabled=false, emitterName=FN_Sprite, simTarget=GPUComputeSim, rendererClasses [&NiagaraSpriteRendererProperties]}
  {bEnabled=false, emitterName=FN_Sprite001, simTarget=GPUComputeSim, rendererClasses [&NiagaraSpriteRendererProperties]}
  {
   bEnabled=false
   emitterName=FN_Sprite002
   simTarget=GPUComputeSim
   rendererClasses [&NiagaraSpriteRendererProperties, &NiagaraMeshRendererProperties]
  }
  {emitterName=FN_Sprite003, rendererClasses [&NiagaraSpriteRendererProperties]}
 ]
 userVariables [
  User.End Position {type=NiagaraFloat, defaultValue=1, description=""}
  User.IsActive {type=NiagaraBool, defaultValue=true}
  User.LandscapeDataInterface {
   type=NiagaraDataInterfaceLandscape
   description=""
   defaultValue {
    struct=NiagaraExt_VariableValue_DataInterface
    value {
     dataInterface=&self:NiagaraDataInterfaceLandscape_0
     dataInterfaceClass=&NiagaraDataInterfaceLandscape
    }
   }
  }
  User.MinPosition {type=NiagaraFloat, defaultValue=0, description=""}
  User.Progress {type=NiagaraFloat, defaultValue=0, description=""}
  User.Spline {
   type=NiagaraDataInterfaceSpline
   description=""
   defaultValue {
    struct=NiagaraExt_VariableValue_DataInterface
    value {
     dataInterface=&self:NiagaraDataInterfaceSpline_0
     dataInterfaceClass=/Script/Niagara.NiagaraDataInterfaceSpline
    }
   }
  }
  User.Start Position {type=NiagaraFloat, defaultValue=0, description=""}
 ]
}
